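# Editor utility widget for a clickable column header in the Sort Level Materials tool of DelMar's track performance review tools.
# An overlay holds a button with a text label; PreConstruct sets the label from HeaderName and the button tooltip to "Sort by" that name.
# Clicking the button broadcasts the OnClickedSort dispatcher.
# An OnClickedVisibility dispatcher is declared but never called here.
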EditorUtilityWidgetBlueprint /DelMarTools/EditorUtilities/AssetReviewTools/TrackPerformanceReview/Subwidgets/SortLevelMaterials/SortLevelMaterialToolColumnHeader
graphs [
 EventGraph {nodes [
   "On Clicked (HeaderButton)"() {is=title K2Node_ComponentBoundEvent}
   > Call On Clicked Sort!(execute:Exec, self:Sort Level Material Tool Column Header Object Reference) -> (then:Exec) {is=title K2Node_CallDelegate}
   Event PreConstruct()
   > "SetText (Text)"(self=Get HeaderText() -> (HeaderText:Text Object Reference), InText=Get HeaderName() -> (HeaderName:Text))
   > Set ToolTipText!(
    execute
    ToolTipText:Text=Format Text!(Format="INVTEXT(\"Sort by {name}\")", name:Text=Get HeaderName()) -> (Result)
    self:Widget Object Reference=Get HeaderButton() -> (HeaderButton:Editor Utility Button Object Reference)
   ) -> (then, Output_Get:Text)
  ]}
 OnClickedSort {nodes [OnClickedSort() {is=title K2Node_FunctionEntry}]}
 OnClickedVisibility {nodes [OnClickedVisibility() {is=title K2Node_FunctionEntry}]}
]
variables [
 HeaderName {type=string, default=Header Name}
 OnClickedSort {default=()}
 OnClickedVisibility {default=()}
]
widgets {
 info {rootWidgetClass=&Overlay, widgetCount=3}
 widgets [
  Overlay_40 {widgetClassPath=&Overlay, widget=Overlay_40}
  HeaderButton {
   widgetClassPath=&EditorUtilityButton
   parent=Overlay_40
   widget=HeaderButton
   slot {class=OverlaySlot, properties {horizontalAlignment=HAlign_Fill, verticalAlignment=VAlign_Fill}}
  }
  HeaderText {
   widgetClassPath=&TextBlock
   bIsVariable=true
   parent=HeaderButton
   widget=HeaderText
   properties {
    text=Header
    colorAndOpacity {colorUseRule=UseColor_Specified, specifiedColor=<rgba 0.5271 0.5271 0.5271 1>}
    font {
     bForceMonospaced=false
     bMaterialIsStencil=false
     fontMaterial=null
     fontObject=&Roboto
     letterSpacing=0
     monospacedWidth=1
     size=9
     skewAmount=0
     typefaceFontName=Bold
     outlineSettings {
      bApplyOutlineToDropShadows=false
      bMiteredCorners=false
      bSeparateFillAlpha=false
      outlineMaterial=null
      outlineSize=0
      outlineColor=<rgba 0 0 0 1>
     }
    }
   }
   slot {class=ButtonSlot}
  }
 ]
}
