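# Editor utility widget for one row in the DelMar track performance review's level material sorting tool, showing a material's cost and usage.
# PreConstruct fills read-only columns from MaterialReference and the Info struct: name, shader statistics and actor count.
# FindMaterialButton opens the material's editor, while FindActorsButton broadcasts OnClickFindActors with the row itself for external handling.
# A SelectObjectsInContentBrowser helper syncs the browser to given objects but is not called here.
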EditorUtilityWidgetBlueprint /DelMarTools/EditorUtilities/AssetReviewTools/TrackPerformanceReview/Subwidgets/SortLevelMaterials/SortLevelMaterialsRow
graphs [
 EventGraph {nodes [
   "On Clicked (FindMaterialButton)"() {is=title K2Node_ComponentBoundEvent}
   > OpenEditorForAssets!(
    execute
    self:Asset Editor Subsystem Object Reference=Get AssetEditorSubsystem()
    Assets:Array of Object References=Make Array("[0]":Object Reference=Get MaterialReference() -> (MaterialReference:Material Interface Object Reference)) -> (Array:Array of Object References)
    OpenedMethod:EAssetTypeActivationOpenedMethod Enum=Edit
   ) -> (then, ReturnValue:bool)
   "On Clicked (FindActorsButton)"() {is=title K2Node_ComponentBoundEvent}
   > Call On Click Find Actors!(execute:Exec, self:Sort Level Materials Row Object Reference, SelfReference:Sort Level Materials Row Object Reference=Self-Reference()) -> (then:Exec) {is=title K2Node_CallDelegate}
   Event PreConstruct()
   n1 = Get MaterialReference()
   > Set Text(
    Text="To Text (String)"(InString=GetDisplayName(Object=@n1))
    self:Editable Text Object Reference=Get MaterialNameText() -> (MaterialNameText:Editor Utility Editable Text Object Reference)
   )
   n2 = Get!() -> (
    Info_Actors_6_#1:Array of Actor Object References
    Info_BuildingActors_14_#2:Array of Building Level Instance Object References
    Info_Statistics_10_#3_NumVertexShaderInstructions:int
    Info_Statistics_10_#3_NumPixelShaderInstructions:int
    Info_Statistics_10_#3_NumSamplers:int
    Info_Statistics_10_#3_NumVertexTextureSamples:int
    Info_Statistics_10_#3_NumPixelTextureSamples:int
    Info_Statistics_10_#3_NumVirtualTextureSamples:int
    Info_Statistics_10_#3_NumUVScalars:int
    Info_Statistics_10_#3_NumInterpolatorScalars:int
   )
   > Set Text(
    Text="To Text (Integer)"(Value=@n2.Info_Statistics_10_#3_NumVertexShaderInstructions)
    self=Get VSInstructionsText() -> (VSInstructionsText:Editor Utility Editable Text Object Reference)
   )
   > Set Text(
    Text="To Text (Integer)"(Value=@n2.Info_Statistics_10_#3_NumPixelShaderInstructions)
    self=Get PSInstructionsText() -> (PSInstructionsText:Editor Utility Editable Text Object Reference)
   )
   > Set Text(
    Text="To Text (Integer)"(Value=@n2.Info_Statistics_10_#3_NumSamplers)
    self=Get SamplersText() -> (SamplersText:Editor Utility Editable Text Object Reference)
   )
   > Set Text(
    Text="To Text (Integer)"(Value=@n2.Info_Statistics_10_#3_NumVertexTextureSamples)
    self=Get VertexTextureSamplesText() -> (VertexTextureSamplesText:Editor Utility Editable Text Object Reference)
   )
   > Set Text(
    Text="To Text (Integer)"(Value=@n2.Info_Statistics_10_#3_NumPixelTextureSamples)
    self=Get PixelTextureSamplesText() -> (PixelTextureSamplesText:Editor Utility Editable Text Object Reference)
   )
   > Set Text(
    Text="To Text (Integer)"(Value=@n2.Info_Statistics_10_#3_NumVirtualTextureSamples)
    self=Get VirtualTextureSamplesText() -> (VirtualTextureSamplesText:Editor Utility Editable Text Object Reference)
   )
   > Set Text(
    Text="To Text (Integer)"(Value=@n2.Info_Statistics_10_#3_NumUVScalars)
    self=Get UVScalarsText() -> (UVScalarsText:Editor Utility Editable Text Object Reference)
   )
   > Set Text(
    Text="To Text (Integer)"(Value=@n2.Info_Statistics_10_#3_NumInterpolatorScalars)
    self=Get InterpolatorScalarsText() -> (InterpolatorScalarsText:Editor Utility Editable Text Object Reference)
   )
   > Set Text(
    Text="To Text (Integer)"(Value=Length(TargetArray=@n2))
    self=Get ActorsinLevelText() -> (ActorsinLevelText:Editor Utility Editable Text Object Reference)
   )
  ]}
 OnClickFindActors {nodes [OnClickFindActors!() -> (then, SelfReference:Sort Level Materials Row Object Reference) {is=title K2Node_FunctionEntry}]}
 SelectObjectsInContentBrowser {nodes [
   SelectObjectsInContentBrowser!() -> (then, InObjectsToSelect:Array of Object References) {is=title K2Node_FunctionEntry}
   > n1 = Sequence()
   > n2 = For Each Loop(Array:Array of Object References=Get InObjectsToSelect() -> (InObjectsToSelect:Array of Object References)) -> (Array Element:Object Reference)
   > Add(TargetArray=Get AssetPaths() -> (AssetPaths:Array of Strings), NewItem=Get Object Path String(Object=@n2)) {is=title K2Node_CallArrayFunction}
   SyncBrowserToObjects(execute=@n1.then_1, AssetPaths=Get AssetPaths())
  ]}
]
variables [
 Info {
  type=SortTool_LevelMaterialInfoStruct
  default {
   actors []
   buildingActors []
   statistics {
    numInterpolatorScalars=0
    numPixelShaderInstructions=0
    numPixelTextureSamples=0
    numSamplers=0
    numUVScalars=0
    numVertexShaderInstructions=0
    numVertexTextureSamples=0
    numVirtualTextureSamples=0
   }
  }
 }
 MaterialReference {type=MaterialInterface, default=null}
 OnClickFindActors {default=()}
]
widgets {
 info {rootWidgetClass=&VerticalBox, widgetCount=36}
 widgets [
  VerticalBox_55 {widgetClassPath=&VerticalBox, widget=VerticalBox_55}
  HorizontalBox_26 {widgetClassPath=&HorizontalBox, parent=VerticalBox_55, widget=HorizontalBox_26, slot {class=VerticalBoxSlot}}
  VerticalDivider_8 {
   widgetClassPath=&Image
   parent=HorizontalBox_26
   widget=VerticalDivider_8
   properties {brush {
     bIsDynamicallyLoaded=false
     drawAs=Image
     imageType=NoImage
     mirroring=NoMirror
     resourceName=None
     resourceObject=null
     tiling=NoTile
     imageSize=<1 32>
     margin=<ltrb 0 0 0 0>
     outlineSettings {
      bUseBrushTransparency=false
      roundingType=HalfHeightRadius
      width=0
      color {colorUseRule=UseColor_Specified, specifiedColor=<rgba 0 0 0 0>}
      cornerRadii=<0 0 0 0>
     }
     tintColor {colorUseRule=UseColor_Specified, specifiedColor=<rgba 0.0284 0.0284 0.0284 1>}
     uVRegion {bIsValid=false, max=<0 0>, min=<0 0>}
    }}
   slot {class=HorizontalBoxSlot, properties {horizontalAlignment=HAlign_Center}}
  }
  VerticalBox_1 {
   widgetClassPath=&VerticalBox
   parent=HorizontalBox_26
   widget=VerticalBox_1
   slot {class=HorizontalBoxSlot, properties {padding=<ltrb 10 10 10 10>, size {sizeRule=Fill, value=1}}}
  }
  MaterialNameText {
   widgetClassPath=/Script/Blutility.EditorUtilityEditableText
   bIsVariable=true
   parent=VerticalBox_1
   widget=MaterialNameText
   properties {isReadOnly=true, text=Material Name}
   slot {class=VerticalBoxSlot, properties {horizontalAlignment=HAlign_Center}}
  }
  FindMaterialButton {
   widgetClassPath=&Button
   parent=VerticalBox_1
   widget=FindMaterialButton
   properties {
    backgroundColor=<rgba 0 0 0 0.25>
    widgetStyle {
     clickedSlateSound {resourceObject=null}
     disabled {
      bIsDynamicallyLoaded=false
      drawAs=RoundedBox
      imageType=NoImage
      mirroring=NoMirror
      resourceName=None
      resourceObject=null
      tiling=NoTile
      imageSize=<0 0>
      margin=<ltrb 0 0 0 0>
      outlineSettings {
       bUseBrushTransparency=false
       roundingType=FixedRadius
       width=1
       color {colorUseRule=UseColor_ColorTable, specifiedColor=<rgba 0 0 0 0>}
       cornerRadii=<4 4 4 4>
      }
      tintColor {colorUseRule=UseColor_Specified, specifiedColor=<rgba 0.0395 0.0395 0.0395 1>}
      uVRegion {bIsValid=false, max=<0 0>, min=<0 0>}
     }
     disabledForeground {colorUseRule=UseColor_ColorTable, specifiedColor=<rgba 0 0 0 0>}
     hovered {
      bIsDynamicallyLoaded=false
      drawAs=RoundedBox
      imageType=NoImage
      mirroring=NoMirror
      resourceName=None
      resourceObject=null
      tiling=NoTile
      imageSize=<32 32>
      margin=<ltrb 0 0 0 0>
      outlineSettings {
       bUseBrushTransparency=false
       roundingType=FixedRadius
       width=1
       color {colorUseRule=UseColor_ColorTable, specifiedColor=<rgba 0 0 0 0>}
       cornerRadii=<4 4 4 4>
      }
      tintColor {colorUseRule=UseColor_Specified, specifiedColor=<rgba 0.0953 0.0953 0.0953 1>}
      uVRegion {bIsValid=false, max=<0 0>, min=<0 0>}
     }
     hoveredForeground {colorUseRule=UseColor_ColorTable, specifiedColor=<rgba 0 0 0 0>}
     hoveredSlateSound {resourceObject=null}
     normal {
      bIsDynamicallyLoaded=false
      drawAs=RoundedBox
      imageType=NoImage
      mirroring=NoMirror
      resourceName=None
      resourceObject=null
      tiling=NoTile
      imageSize=<32 32>
      margin=<ltrb 0 0 0 0>
      outlineSettings {
       bUseBrushTransparency=false
       roundingType=FixedRadius
       width=1
       color {colorUseRule=UseColor_ColorTable, specifiedColor=<rgba 0 0 0 0>}
       cornerRadii=<4 4 4 4>
      }
      tintColor {colorUseRule=UseColor_Specified, specifiedColor=<rgba 0.0395 0.0395 0.0395 1>}
      uVRegion {bIsValid=false, max=<0 0>, min=<0 0>}
     }
     normalForeground {colorUseRule=UseColor_ColorTable, specifiedColor=<rgba 0 0 0 0>}
     normalPadding=<ltrb 12 1.5 12 1.5>
     pressed {
      bIsDynamicallyLoaded=false
      drawAs=RoundedBox
      imageType=NoImage
      mirroring=NoMirror
      resourceName=None
      resourceObject=null
      tiling=NoTile
      imageSize=<32 32>
      margin=<ltrb 0 0 0 0>
      outlineSettings {
       bUseBrushTransparency=false
       roundingType=FixedRadius
       width=1
       color {colorUseRule=UseColor_ColorTable, specifiedColor=<rgba 0 0 0 0>}
       cornerRadii=<4 4 4 4>
      }
      tintColor {colorUseRule=UseColor_Specified, specifiedColor=<rgba 0.0284 0.0284 0.0284 1>}
      uVRegion {bIsValid=false, max=<0 0>, min=<0 0>}
     }
     pressedForeground {colorUseRule=UseColor_ColorTable, specifiedColor=<rgba 0 0 0 0>}
     pressedPadding=<ltrb 12 2.5 12 0.5>
     pressedSlateSound {resourceObject=null}
    }
   }
   slot {class=VerticalBoxSlot, properties {horizontalAlignment=HAlign_Center, verticalAlignment=VAlign_Bottom}}
  }
  SizeBox_0 {
   widgetClassPath=&SizeBox
   parent=FindMaterialButton
   widget=SizeBox_0
   properties {bOverride_HeightOverride=true, bOverride_WidthOverride=true, heightOverride=28, widthOverride=28}
   slot {class=ButtonSlot, properties {padding=<ltrb 0 0 0 0>}}
  }
  Overlay_0 {widgetClassPath=&Overlay, parent=SizeBox_0, widget=Overlay_0, slot {class=SizeBoxSlot}}
  Image_386 {
   widgetClassPath=&Image
   parent=Overlay_0
   widget=Image_386
   properties {
    brush {
     bIsDynamicallyLoaded=false
     drawAs=Image
     imageType=NoImage
     mirroring=NoMirror
     resourceName=None
     resourceObject=/Game/UI/Friends_UI/Social/Search
     tiling=NoTile
     imageSize=<20 20>
     margin=<ltrb 0 0 0 0>
     outlineSettings {
      bUseBrushTransparency=false
      roundingType=HalfHeightRadius
      width=0
      color {colorUseRule=UseColor_Specified, specifiedColor=<rgba 0 0 0 0>}
      cornerRadii=<0 0 0 0>
     }
     tintColor {colorUseRule=UseColor_Specified, specifiedColor=<rgba 1 1 1 1>}
     uVRegion {bIsValid=false, max=<0 0>, min=<0 0>}
    }
    colorAndOpacity=<rgba 0.3906 0.3906 0.3906 1>
   }
   slot {class=OverlaySlot, properties {horizontalAlignment=HAlign_Right}}
  }
  Image {
   widgetClassPath=&Image
   parent=Overlay_0
   widget=Image
   properties {
    clipping=ClipToBounds
    brush {
     bIsDynamicallyLoaded=false
     drawAs=Image
     imageType=NoImage
     mirroring=NoMirror
     resourceName=None
     resourceObject=/Game/UI/Foundation/Textures/Icons/Replay/T-Icon-Replay-Folder-L
     tiling=NoTile
     imageSize=<20 20>
     margin=<ltrb 0 0 0 0>
     outlineSettings {
      bUseBrushTransparency=false
      roundingType=HalfHeightRadius
      width=0
      color {colorUseRule=UseColor_Specified, specifiedColor=<rgba 0 0 0 0>}
      cornerRadii=<0 0 0 0>
     }
     tintColor {colorUseRule=UseColor_Specified, specifiedColor=<rgba 1 1 1 1>}
     uVRegion {bIsValid=false, max=<0 0>, min=<0 0>}
    }
    renderTransform {angle=0, scale=<1 1>, shear=<0 0>, translation=<0 3>}
   }
   slot {class=OverlaySlot, properties {verticalAlignment=VAlign_Bottom}}
  }
  VerticalDivider_7 {
   widgetClassPath=&Image
   parent=HorizontalBox_26
   widget=VerticalDivider_7
   properties {brush {
     bIsDynamicallyLoaded=false
     drawAs=Image
     imageType=NoImage
     mirroring=NoMirror
     resourceName=None
     resourceObject=null
     tiling=NoTile
     imageSize=<1 32>
     margin=<ltrb 0 0 0 0>
     outlineSettings {
      bUseBrushTransparency=false
      roundingType=HalfHeightRadius
      width=0
      color {colorUseRule=UseColor_Specified, specifiedColor=<rgba 0 0 0 0>}
      cornerRadii=<0 0 0 0>
     }
     tintColor {colorUseRule=UseColor_Specified, specifiedColor=<rgba 0.0284 0.0284 0.0284 1>}
     uVRegion {bIsValid=false, max=<0 0>, min=<0 0>}
    }}
   slot {class=HorizontalBoxSlot, properties {horizontalAlignment=HAlign_Center}}
  }
  VSInstructionsText {
   widgetClassPath=&EditorUtilityEditableText
   bIsVariable=true
   parent=HorizontalBox_26
   widget=VSInstructionsText
   properties {isReadOnly=true, text=0}
   slot {
    class=HorizontalBoxSlot
    properties {horizontalAlignment=HAlign_Center, verticalAlignment=VAlign_Center, padding=<ltrb 10 10 10 10>, size {sizeRule=Fill, value=1}}
   }
  }
  VerticalDivider {
   widgetClassPath=&Image
   parent=HorizontalBox_26
   widget=VerticalDivider
   properties {brush {
     bIsDynamicallyLoaded=false
     drawAs=Image
     imageType=NoImage
     mirroring=NoMirror
     resourceName=None
     resourceObject=null
     tiling=NoTile
     imageSize=<1 32>
     margin=<ltrb 0 0 0 0>
     outlineSettings {
      bUseBrushTransparency=false
      roundingType=HalfHeightRadius
      width=0
      color {colorUseRule=UseColor_Specified, specifiedColor=<rgba 0 0 0 0>}
      cornerRadii=<0 0 0 0>
     }
     tintColor {colorUseRule=UseColor_Specified, specifiedColor=<rgba 0.0284 0.0284 0.0284 1>}
     uVRegion {bIsValid=false, max=<0 0>, min=<0 0>}
    }}
   slot {class=HorizontalBoxSlot, properties {horizontalAlignment=HAlign_Center}}
  }
  PSInstructionsText {
   widgetClassPath=&EditorUtilityEditableText
   bIsVariable=true
   parent=HorizontalBox_26
   widget=PSInstructionsText
   properties {isReadOnly=true, text=0}
   slot {
    class=HorizontalBoxSlot
    properties {horizontalAlignment=HAlign_Center, verticalAlignment=VAlign_Center, padding=<ltrb 10 10 10 10>, size {sizeRule=Fill, value=1}}
   }
  }
  VerticalDivider_1 {
   widgetClassPath=&Image
   parent=HorizontalBox_26
   widget=VerticalDivider_1
   properties {brush {
     bIsDynamicallyLoaded=false
     drawAs=Image
     imageType=NoImage
     mirroring=NoMirror
     resourceName=None
     resourceObject=null
     tiling=NoTile
     imageSize=<1 32>
     margin=<ltrb 0 0 0 0>
     outlineSettings {
      bUseBrushTransparency=false
      roundingType=HalfHeightRadius
      width=0
      color {colorUseRule=UseColor_Specified, specifiedColor=<rgba 0 0 0 0>}
      cornerRadii=<0 0 0 0>
     }
     tintColor {colorUseRule=UseColor_Specified, specifiedColor=<rgba 0.0284 0.0284 0.0284 1>}
     uVRegion {bIsValid=false, max=<0 0>, min=<0 0>}
    }}
   slot {class=HorizontalBoxSlot, properties {horizontalAlignment=HAlign_Center}}
  }
  SamplersText {
   widgetClassPath=&EditorUtilityEditableText
   bIsVariable=true
   parent=HorizontalBox_26
   widget=SamplersText
   properties {isReadOnly=true, text=0}
   slot {
    class=HorizontalBoxSlot
    properties {horizontalAlignment=HAlign_Center, verticalAlignment=VAlign_Center, padding=<ltrb 10 10 10 10>, size {sizeRule=Fill, value=1}}
   }
  }
  VerticalDivider_2 {
   widgetClassPath=&Image
   parent=HorizontalBox_26
   widget=VerticalDivider_2
   properties {brush {
     bIsDynamicallyLoaded=false
     drawAs=Image
     imageType=NoImage
     mirroring=NoMirror
     resourceName=None
     resourceObject=null
     tiling=NoTile
     imageSize=<1 32>
     margin=<ltrb 0 0 0 0>
     outlineSettings {
      bUseBrushTransparency=false
      roundingType=HalfHeightRadius
      width=0
      color {colorUseRule=UseColor_Specified, specifiedColor=<rgba 0 0 0 0>}
      cornerRadii=<0 0 0 0>
     }
     tintColor {colorUseRule=UseColor_Specified, specifiedColor=<rgba 0.0284 0.0284 0.0284 1>}
     uVRegion {bIsValid=false, max=<0 0>, min=<0 0>}
    }}
   slot {class=HorizontalBoxSlot, properties {horizontalAlignment=HAlign_Center}}
  }
  VertexTextureSamplesText {
   widgetClassPath=&EditorUtilityEditableText
   bIsVariable=true
   parent=HorizontalBox_26
   widget=VertexTextureSamplesText
   properties {isReadOnly=true, text=0}
   slot {
    class=HorizontalBoxSlot
    properties {horizontalAlignment=HAlign_Center, verticalAlignment=VAlign_Center, padding=<ltrb 10 10 10 10>, size {sizeRule=Fill, value=1}}
   }
  }
  VerticalDivider_3 {
   widgetClassPath=&Image
   parent=HorizontalBox_26
   widget=VerticalDivider_3
   properties {brush {
     bIsDynamicallyLoaded=false
     drawAs=Image
     imageType=NoImage
     mirroring=NoMirror
     resourceName=None
     resourceObject=null
     tiling=NoTile
     imageSize=<1 32>
     margin=<ltrb 0 0 0 0>
     outlineSettings {
      bUseBrushTransparency=false
      roundingType=HalfHeightRadius
      width=0
      color {colorUseRule=UseColor_Specified, specifiedColor=<rgba 0 0 0 0>}
      cornerRadii=<0 0 0 0>
     }
     tintColor {colorUseRule=UseColor_Specified, specifiedColor=<rgba 0.0284 0.0284 0.0284 1>}
     uVRegion {bIsValid=false, max=<0 0>, min=<0 0>}
    }}
   slot {class=HorizontalBoxSlot, properties {horizontalAlignment=HAlign_Center}}
  }
  PixelTextureSamplesText {
   widgetClassPath=&EditorUtilityEditableText
   bIsVariable=true
   parent=HorizontalBox_26
   widget=PixelTextureSamplesText
   properties {isReadOnly=true, text=0}
   slot {
    class=HorizontalBoxSlot
    properties {horizontalAlignment=HAlign_Center, verticalAlignment=VAlign_Center, padding=<ltrb 10 10 10 10>, size {sizeRule=Fill, value=1}}
   }
  }
  VerticalDivider_4 {
   widgetClassPath=&Image
   parent=HorizontalBox_26
   widget=VerticalDivider_4
   properties {brush {
     bIsDynamicallyLoaded=false
     drawAs=Image
     imageType=NoImage
     mirroring=NoMirror
     resourceName=None
     resourceObject=null
     tiling=NoTile
     imageSize=<1 32>
     margin=<ltrb 0 0 0 0>
     outlineSettings {
      bUseBrushTransparency=false
      roundingType=HalfHeightRadius
      width=0
      color {colorUseRule=UseColor_Specified, specifiedColor=<rgba 0 0 0 0>}
      cornerRadii=<0 0 0 0>
     }
     tintColor {colorUseRule=UseColor_Specified, specifiedColor=<rgba 0.0284 0.0284 0.0284 1>}
     uVRegion {bIsValid=false, max=<0 0>, min=<0 0>}
    }}
   slot {class=HorizontalBoxSlot, properties {horizontalAlignment=HAlign_Center}}
  }
  VirtualTextureSamplesText {
   widgetClassPath=&EditorUtilityEditableText
   bIsVariable=true
   parent=HorizontalBox_26
   widget=VirtualTextureSamplesText
   properties {isReadOnly=true, text=0}
   slot {
    class=HorizontalBoxSlot
    properties {horizontalAlignment=HAlign_Center, verticalAlignment=VAlign_Center, padding=<ltrb 10 10 10 10>, size {sizeRule=Fill, value=1}}
   }
  }
  VerticalDivider_5 {
   widgetClassPath=&Image
   parent=HorizontalBox_26
   widget=VerticalDivider_5
   properties {brush {
     bIsDynamicallyLoaded=false
     drawAs=Image
     imageType=NoImage
     mirroring=NoMirror
     resourceName=None
     resourceObject=null
     tiling=NoTile
     imageSize=<1 32>
     margin=<ltrb 0 0 0 0>
     outlineSettings {
      bUseBrushTransparency=false
      roundingType=HalfHeightRadius
      width=0
      color {colorUseRule=UseColor_Specified, specifiedColor=<rgba 0 0 0 0>}
      cornerRadii=<0 0 0 0>
     }
     tintColor {colorUseRule=UseColor_Specified, specifiedColor=<rgba 0.0284 0.0284 0.0284 1>}
     uVRegion {bIsValid=false, max=<0 0>, min=<0 0>}
    }}
   slot {class=HorizontalBoxSlot, properties {horizontalAlignment=HAlign_Center}}
  }
  UVScalarsText {
   widgetClassPath=&EditorUtilityEditableText
   bIsVariable=true
   parent=HorizontalBox_26
   widget=UVScalarsText
   properties {isReadOnly=true, text=0}
   slot {
    class=HorizontalBoxSlot
    properties {horizontalAlignment=HAlign_Center, verticalAlignment=VAlign_Center, padding=<ltrb 10 10 10 10>, size {sizeRule=Fill, value=1}}
   }
  }
  VerticalDivider_6 {
   widgetClassPath=&Image
   parent=HorizontalBox_26
   widget=VerticalDivider_6
   properties {brush {
     bIsDynamicallyLoaded=false
     drawAs=Image
     imageType=NoImage
     mirroring=NoMirror
     resourceName=None
     resourceObject=null
     tiling=NoTile
     imageSize=<1 32>
     margin=<ltrb 0 0 0 0>
     outlineSettings {
      bUseBrushTransparency=false
      roundingType=HalfHeightRadius
      width=0
      color {colorUseRule=UseColor_Specified, specifiedColor=<rgba 0 0 0 0>}
      cornerRadii=<0 0 0 0>
     }
     tintColor {colorUseRule=UseColor_Specified, specifiedColor=<rgba 0.0284 0.0284 0.0284 1>}
     uVRegion {bIsValid=false, max=<0 0>, min=<0 0>}
    }}
   slot {class=HorizontalBoxSlot, properties {horizontalAlignment=HAlign_Center}}
  }
  InterpolatorScalarsText {
   widgetClassPath=&EditorUtilityEditableText
   bIsVariable=true
   parent=HorizontalBox_26
   widget=InterpolatorScalarsText
   properties {isReadOnly=true, text=0}
   slot {
    class=HorizontalBoxSlot
    properties {horizontalAlignment=HAlign_Center, verticalAlignment=VAlign_Center, padding=<ltrb 10 10 10 10>, size {sizeRule=Fill, value=1}}
   }
  }
  VerticalDivider_9 {
   widgetClassPath=&Image
   parent=HorizontalBox_26
   widget=VerticalDivider_9
   properties {brush {
     bIsDynamicallyLoaded=false
     drawAs=Image
     imageType=NoImage
     mirroring=NoMirror
     resourceName=None
     resourceObject=null
     tiling=NoTile
     imageSize=<1 32>
     margin=<ltrb 0 0 0 0>
     outlineSettings {
      bUseBrushTransparency=false
      roundingType=HalfHeightRadius
      width=0
      color {colorUseRule=UseColor_Specified, specifiedColor=<rgba 0 0 0 0>}
      cornerRadii=<0 0 0 0>
     }
     tintColor {colorUseRule=UseColor_Specified, specifiedColor=<rgba 0.0284 0.0284 0.0284 1>}
     uVRegion {bIsValid=false, max=<0 0>, min=<0 0>}
    }}
   slot {class=HorizontalBoxSlot, properties {horizontalAlignment=HAlign_Center}}
  }
  VerticalBox_156 {
   widgetClassPath=&VerticalBox
   parent=HorizontalBox_26
   widget=VerticalBox_156
   slot {class=HorizontalBoxSlot, properties {padding=<ltrb 10 10 10 10>, size {sizeRule=Fill, value=1}}}
  }
  ActorsinLevelText {
   widgetClassPath=&EditorUtilityEditableText
   bIsVariable=true
   parent=VerticalBox_156
   widget=ActorsinLevelText
   properties {isReadOnly=true, text=0}
   slot {class=VerticalBoxSlot, properties {horizontalAlignment=HAlign_Center, verticalAlignment=VAlign_Center, size {sizeRule=Fill, value=1}}}
  }
  FindActorsButton {
   widgetClassPath=&Button
   parent=VerticalBox_156
   widget=FindActorsButton
   properties {
    backgroundColor=<rgba 0 0 0 0.25>
    widgetStyle {
     clickedSlateSound {resourceObject=null}
     disabled {
      bIsDynamicallyLoaded=false
      drawAs=RoundedBox
      imageType=NoImage
      mirroring=NoMirror
      resourceName=None
      resourceObject=null
      tiling=NoTile
      imageSize=<0 0>
      margin=<ltrb 0 0 0 0>
      outlineSettings {
       bUseBrushTransparency=false
       roundingType=FixedRadius
       width=1
       color {colorUseRule=UseColor_ColorTable, specifiedColor=<rgba 0 0 0 0>}
       cornerRadii=<4 4 4 4>
      }
      tintColor {colorUseRule=UseColor_Specified, specifiedColor=<rgba 0.0395 0.0395 0.0395 1>}
      uVRegion {bIsValid=false, max=<0 0>, min=<0 0>}
     }
     disabledForeground {colorUseRule=UseColor_ColorTable, specifiedColor=<rgba 0 0 0 0>}
     hovered {
      bIsDynamicallyLoaded=false
      drawAs=RoundedBox
      imageType=NoImage
      mirroring=NoMirror
      resourceName=None
      resourceObject=null
      tiling=NoTile
      imageSize=<32 32>
      margin=<ltrb 0 0 0 0>
      outlineSettings {
       bUseBrushTransparency=false
       roundingType=FixedRadius
       width=1
       color {colorUseRule=UseColor_ColorTable, specifiedColor=<rgba 0 0 0 0>}
       cornerRadii=<4 4 4 4>
      }
      tintColor {colorUseRule=UseColor_Specified, specifiedColor=<rgba 0.0953 0.0953 0.0953 1>}
      uVRegion {bIsValid=false, max=<0 0>, min=<0 0>}
     }
     hoveredForeground {colorUseRule=UseColor_ColorTable, specifiedColor=<rgba 0 0 0 0>}
     hoveredSlateSound {resourceObject=null}
     normal {
      bIsDynamicallyLoaded=false
      drawAs=RoundedBox
      imageType=NoImage
      mirroring=NoMirror
      resourceName=None
      resourceObject=null
      tiling=NoTile
      imageSize=<32 32>
      margin=<ltrb 0 0 0 0>
      outlineSettings {
       bUseBrushTransparency=false
       roundingType=FixedRadius
       width=1
       color {colorUseRule=UseColor_ColorTable, specifiedColor=<rgba 0 0 0 0>}
       cornerRadii=<4 4 4 4>
      }
      tintColor {colorUseRule=UseColor_Specified, specifiedColor=<rgba 0.0395 0.0395 0.0395 1>}
      uVRegion {bIsValid=false, max=<0 0>, min=<0 0>}
     }
     normalForeground {colorUseRule=UseColor_ColorTable, specifiedColor=<rgba 0 0 0 0>}
     normalPadding=<ltrb 12 1.5 12 1.5>
     pressed {
      bIsDynamicallyLoaded=false
      drawAs=RoundedBox
      imageType=NoImage
      mirroring=NoMirror
      resourceName=None
      resourceObject=null
      tiling=NoTile
      imageSize=<32 32>
      margin=<ltrb 0 0 0 0>
      outlineSettings {
       bUseBrushTransparency=false
       roundingType=FixedRadius
       width=1
       color {colorUseRule=UseColor_ColorTable, specifiedColor=<rgba 0 0 0 0>}
       cornerRadii=<4 4 4 4>
      }
      tintColor {colorUseRule=UseColor_Specified, specifiedColor=<rgba 0.0284 0.0284 0.0284 1>}
      uVRegion {bIsValid=false, max=<0 0>, min=<0 0>}
     }
     pressedForeground {colorUseRule=UseColor_ColorTable, specifiedColor=<rgba 0 0 0 0>}
     pressedPadding=<ltrb 12 2.5 12 0.5>
     pressedSlateSound {resourceObject=null}
    }
   }
   slot {class=VerticalBoxSlot, properties {horizontalAlignment=HAlign_Center, verticalAlignment=VAlign_Bottom}}
  }
  FindObjectsButtonIconSizeBox {
   widgetClassPath=&SizeBox
   parent=FindActorsButton
   widget=FindObjectsButtonIconSizeBox
   properties {bOverride_HeightOverride=true, bOverride_WidthOverride=true, heightOverride=28, widthOverride=28}
   slot {class=ButtonSlot, properties {padding=<ltrb 0 0 0 0>}}
  }
  Overlay {widgetClassPath=&Overlay, parent=FindObjectsButtonIconSizeBox, widget=Overlay, slot {class=SizeBoxSlot}}
  Image_1 {
   widgetClassPath=&Image
   parent=Overlay
   widget=Image_1
   properties {
    brush {
     bIsDynamicallyLoaded=false
     drawAs=Image
     imageType=NoImage
     mirroring=NoMirror
     resourceName=None
     resourceObject=&Search
     tiling=NoTile
     imageSize=<20 20>
     margin=<ltrb 0 0 0 0>
     outlineSettings {
      bUseBrushTransparency=false
      roundingType=HalfHeightRadius
      width=0
      color {colorUseRule=UseColor_Specified, specifiedColor=<rgba 0 0 0 0>}
      cornerRadii=<0 0 0 0>
     }
     tintColor {colorUseRule=UseColor_Specified, specifiedColor=<rgba 1 1 1 1>}
     uVRegion {bIsValid=false, max=<0 0>, min=<0 0>}
    }
    colorAndOpacity=<rgba 0.3906 0.3906 0.3906 1>
   }
   slot {class=OverlaySlot, properties {horizontalAlignment=HAlign_Right}}
  }
  Image_2 {
   widgetClassPath=&Image
   parent=Overlay
   widget=Image_2
   properties {
    clipping=ClipToBounds
    brush {
     bIsDynamicallyLoaded=false
     drawAs=Image
     imageType=NoImage
     mirroring=NoMirror
     resourceName=None
     resourceObject=/Engine/Tutorial/Landscape/TutorialAssets/Landscape
     tiling=NoTile
     imageSize=<20 18>
     margin=<ltrb 0 0 0 0>
     outlineSettings {
      bUseBrushTransparency=false
      roundingType=HalfHeightRadius
      width=0
      color {colorUseRule=UseColor_Specified, specifiedColor=<rgba 0 0 0 0>}
      cornerRadii=<0 0 0 0>
     }
     tintColor {colorUseRule=UseColor_Specified, specifiedColor=<rgba 1 1 1 1>}
     uVRegion {bIsValid=false, max=<0 0>, min=<0 0>}
    }
    renderTransform {angle=0, scale=<1 1>, shear=<0 0>, translation=<0 1>}
   }
   slot {class=OverlaySlot, properties {verticalAlignment=VAlign_Bottom}}
  }
  VerticalDivider_10 {
   widgetClassPath=&Image
   parent=HorizontalBox_26
   widget=VerticalDivider_10
   properties {brush {
     bIsDynamicallyLoaded=false
     drawAs=Image
     imageType=NoImage
     mirroring=NoMirror
     resourceName=None
     resourceObject=null
     tiling=NoTile
     imageSize=<1 32>
     margin=<ltrb 0 0 0 0>
     outlineSettings {
      bUseBrushTransparency=false
      roundingType=HalfHeightRadius
      width=0
      color {colorUseRule=UseColor_Specified, specifiedColor=<rgba 0 0 0 0>}
      cornerRadii=<0 0 0 0>
     }
     tintColor {colorUseRule=UseColor_Specified, specifiedColor=<rgba 0.0284 0.0284 0.0284 1>}
     uVRegion {bIsValid=false, max=<0 0>, min=<0 0>}
    }}
   slot {class=HorizontalBoxSlot, properties {horizontalAlignment=HAlign_Center}}
  }
  HorizontalDivider {
   widgetClassPath=&Image
   parent=VerticalBox_55
   widget=HorizontalDivider
   properties {brush {
     bIsDynamicallyLoaded=false
     drawAs=Image
     imageType=NoImage
     mirroring=NoMirror
     resourceName=None
     resourceObject=null
     tiling=NoTile
     imageSize=<1 1>
     margin=<ltrb 0 0 0 0>
     outlineSettings {
      bUseBrushTransparency=false
      roundingType=HalfHeightRadius
      width=0
      color {colorUseRule=UseColor_Specified, specifiedColor=<rgba 0 0 0 0>}
      cornerRadii=<0 0 0 0>
     }
     tintColor {colorUseRule=UseColor_Specified, specifiedColor=<rgba 0.0284 0.0284 0.0284 1>}
     uVRegion {bIsValid=false, max=<0 0>, min=<0 0>}
    }}
   slot {class=VerticalBoxSlot}
  }
 ]
}
