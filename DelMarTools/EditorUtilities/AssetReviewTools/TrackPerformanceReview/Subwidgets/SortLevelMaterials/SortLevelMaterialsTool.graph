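# Editor utility widget for track performance review that lists the current level's materials by shader statistics and the actors using them.
# A delayed async loop steps through level actors, recursing into building level instances, to fill a material-to-info map of stats and actors.
# Headers re-sort the rows, which can select their actors.
# CSV export and cache save/load use executed Python commands for file I/O.
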EditorUtilityWidgetBlueprint /DelMarTools/EditorUtilities/AssetReviewTools/TrackPerformanceReview/Subwidgets/SortLevelMaterials/SortLevelMaterialsTool
graphs [
 Build CSV {nodes [
   n1 = Build CSV!() -> (then, Materials:Array of Material Interface Object References) {is=title K2Node_FunctionEntry}
   > Set CSV String(CSV String:String="Material,VS Instructions,PS Instructions,Samplers,Vertex Texture Samples,Pixel Texture Samples,Virtual Texture Samples,UV Scalars,Interpolator Scalars,Actors in Level") -> (Output_Get:String)
   > n2 = For Each Loop(Array:Array of Material Interface Object References=@n1) -> (Array Element:Material Interface Object Reference)
   n3 = Find!(
    TargetMap:Map of Material Interface Object References to Sort Tool Level Material Info Struct Structures=Get LevelMaterialInfo() -> (LevelMaterialInfo:Map of Material Interface Object References to Sort Tool Level Material Info Struct Structures)
    Key:"Material Interface Object Reference (by ref)"=@n2
   ) -> (
    Value_Actors_6_#1:Array of Actor Object References
    Value_BuildingActors_14_#2:Array of Building Level Instance Object References
    Value_Statistics_10_#3_NumVertexShaderInstructions:int
    Value_Statistics_10_#3_NumPixelShaderInstructions:int
    Value_Statistics_10_#3_NumSamplers:int
    Value_Statistics_10_#3_NumVertexTextureSamples:int
    Value_Statistics_10_#3_NumPixelTextureSamples:int
    Value_Statistics_10_#3_NumVirtualTextureSamples:int
    Value_Statistics_10_#3_NumUVScalars:int
    Value_Statistics_10_#3_NumInterpolatorScalars:int
    ReturnValue
   )
   n4 = Format Text!(
    Format="INVTEXT(\"\\r\\n{Material},{VS Instructions},{PS Instructions},{Samplers},{Vertex Texture Samples},{Pixel Texture Samples},{Virtual Texture Samples},{UV Scalars},{Interpolator Scalars},{Actors in Level}\")"
    Material:String="To String (SoftObjectReference)"!(SoftObjectReference:"Soft Object Reference (by ref)"=Make Soft Reference!(Input:Material Interface Object Reference=@n2) -> (Output:Material Interface Soft Object Reference)) -> (ReturnValue:String)
    VS Instructions:int=@n3.Value_Statistics_10_#3_NumVertexShaderInstructions
    PS Instructions:int=@n3.Value_Statistics_10_#3_NumPixelShaderInstructions
    Samplers:int=@n3.Value_Statistics_10_#3_NumSamplers
    Vertex Texture Samples:int=@n3.Value_Statistics_10_#3_NumVertexTextureSamples
    Pixel Texture Samples:int=@n3.Value_Statistics_10_#3_NumPixelTextureSamples
    Virtual Texture Samples:int=@n3.Value_Statistics_10_#3_NumVirtualTextureSamples
    UV Scalars:int=@n3.Value_Statistics_10_#3_NumUVScalars
    Interpolator Scalars:int=@n3.Value_Statistics_10_#3_NumInterpolatorScalars
    Actors in Level:int=Length(TargetArray=@n3)
   ) -> (Result)
   > Set CSV String(CSV String=Append(A=Get CSV String() -> (CSV String:String), B="To String (Text)"(InText=@n4)))
  ]}
 EventGraph {nodes [
   "On Clicked Sort (VSInstructionsHeader)"() {is=title K2Node_ComponentBoundEvent}
   > n1 = SortMaterials!(execute, self, SortBy:SortTool_SortType Enum=NewEnumerator0) -> (then, SortedMaterials:Array of Material Interface Object References)
   > ConstructListFromArray!(execute, self, Materials:Array of Material Interface Object References=@n1) -> (then)
   "On Clicked Sort (PSInstructionsHeader)"() {is=title K2Node_ComponentBoundEvent}
   > n2 = SortMaterials(SortBy=NewEnumerator1)
   > ConstructListFromArray(Materials=@n2)
   n3 = "On Clicked (FetchMaterialsButton)"() {is=title K2Node_ComponentBoundEvent}
   Set bShouldStop(execute=[@n3.then, @n6.then], bShouldStop:bool=false) -> (Output_Get:bool)
   > SetActiveWidget(
    self:Widget Switcher Object Reference=Get RootSwitcher() -> (RootSwitcher:Widget Switcher Object Reference)
    Widget=Get FetchingMaterialsVerticalBox() -> (FetchingMaterialsVerticalBox:Vertical Box Object Reference)
   )
   > n4 = GetAllLevelActors(self=Get EditorActorSubsystem())
   > Set LevelActors(LevelActors:Array of Actor Object References=@n4) -> (Output_Get:Array of Actor Object References)
   > Set CurrentActorIndex(CurrentActorIndex:int=0) -> (Output_Get:int=0)
   > Clear!(execute, TargetMap:Map of Material Interface Object References to Sort Tool Level Material Info Struct Structures=Get LevelMaterialInfo()) -> (then) {is=title K2Node_CallFunction}
   > AsyncFetchMaterials()
   "On Clicked (CancelButton)"() {is=title K2Node_ComponentBoundEvent}
   > n5 = SetActiveWidget(self=Get RootSwitcher(), Widget=Get StartingOptionsSizeBox() -> (StartingOptionsSizeBox:Size Box Object Reference))
   Set bShouldStop(execute=[@n5.then, @n26.then], bShouldStop=true)
   n6 = "On Clicked (RestartButton)"() {is=title K2Node_ComponentBoundEvent}
   "On Clicked (LoadCachedResultsButton)"() {is=title K2Node_ComponentBoundEvent}
   > n7 = ImportCache!(execute, self) -> (then, ReturnValue:bool)
   > Branch(Condition=@n7)
   > ShowMaterialsList()
   "On Clicked (SaveCacheButton)"() {is=title K2Node_ComponentBoundEvent}
   > SaveCache()
   "On Clicked Sort (SamplersHeader)"() {is=title K2Node_ComponentBoundEvent}
   > n8 = SortMaterials(SortBy=NewEnumerator2)
   > ConstructListFromArray(Materials=@n8)
   "On Clicked Sort (VertexTextureSamplesHeader)"() {is=title K2Node_ComponentBoundEvent}
   > n9 = SortMaterials(SortBy=NewEnumerator3)
   > ConstructListFromArray(Materials=@n9)
   "On Clicked Sort (PixelTextureSamplesHeader)"() {is=title K2Node_ComponentBoundEvent}
   > n10 = SortMaterials(SortBy=NewEnumerator4)
   > ConstructListFromArray(Materials=@n10)
   "On Clicked Sort (VirtualTextureSamplesHeader)"() {is=title K2Node_ComponentBoundEvent}
   > n11 = SortMaterials(SortBy=NewEnumerator5)
   > ConstructListFromArray(Materials=@n11)
   "On Clicked Sort (UVScalarsHeader)"() {is=title K2Node_ComponentBoundEvent}
   > n12 = SortMaterials(SortBy=NewEnumerator6)
   > ConstructListFromArray(Materials=@n12)
   "On Clicked Sort (InterpolatorScalarsHeader)"() {is=title K2Node_ComponentBoundEvent}
   > n13 = SortMaterials(SortBy=NewEnumerator7)
   > ConstructListFromArray(Materials=@n13)
   "On Clicked (ExportCSVButton)"() {is=title K2Node_ComponentBoundEvent}
   > Export CSV()
   "On Clicked Sort (ActorsinLevelHeader)"() {is=title K2Node_ComponentBoundEvent}
   > n14 = SortMaterials(SortBy=NewEnumerator8)
   > ConstructListFromArray(Materials=@n14)
   ShowMaterialsList() {is=title K2Node_CustomEvent}
   > n15 = Keys!(execute, TargetMap:Map of Material Interface Object References to Sort Tool Level Material Info Struct Structures=Get LevelMaterialInfo()) -> (then, Keys:Array of Material Interface Object References)
   > n16 = Set MaterialsArray(MaterialsArray:Array of Material Interface Object References=@n15) -> (Output_Get:Array of Material Interface Object References)
   > n17 = Branch(Condition="integer > integer"(A=Length(TargetArray:Array of Material Interface Object References=@n16)))
   > ConstructListFromArray(Materials=Get MaterialsArray() -> (MaterialsArray:Array of Material Interface Object References))
   SetActiveWidget(execute=@n17.else, self=Get RootSwitcher(), Widget=Get NoMaterialsVerticalBox() -> (NoMaterialsVerticalBox:Vertical Box Object Reference))
   AsyncFetchMaterials() {is=title K2Node_CustomEvent}
   > n18 = Branch(Condition=Get bShouldStop() -> (bShouldStop:bool))
   n19 = AsyncEditorDelay!(execute:Exec=@n18.else, Seconds:float=0.008, MinimumFrames:int=0) -> (then:Exec, Complete:Exec) {is=title K2Node_AsyncAction}
   FetchNextActorMaterials(execute=@n19.Complete)
   n20 = ConstructListFromArray!() -> (OutputDelegate, then, Materials:Array of Material Interface Object References) {is=title K2Node_CustomEvent}
   > n21 = Sequence()
   > ClearChildren(self=Get MaterialListScrollBox() -> (MaterialListScrollBox:Editor Utility Scroll Box Object Reference))
   > n22 = For Each Loop(Array=@n20.Materials)
   SetActiveWidget(
    execute=@n22.Completed
    self=Get RootSwitcher()
    Widget=Get MaterialListVerticalBox() -> (MaterialListVerticalBox:Vertical Box Object Reference)
   )
   n23 = Construct Object from Class!(
    execute=@n22.LoopBody
    Class=&/SortLevelMaterialsRow.SortLevelMaterialsRow_C
    self
    MaterialReference:Material Interface Object Reference=@n22
    Info:Sort Tool Level Material Info Struct Structure=Find!(TargetMap=Get LevelMaterialInfo(), Key=@n22) -> (Value:Sort Tool Level Material Info Struct Structure, ReturnValue)
   ) -> (then, ReturnValue:Sort Level Materials Row Object Reference)
   > AddChild(self=Get MaterialListScrollBox(), Content=@n23)
   > Bind Event to On Click Find Actors(self:Sort Level Materials Row Object Reference=@n23, Delegate=Create Event()) {is=title K2Node_AddDelegate}
   Build CSV!(execute=@n21.then_1, self, Materials:Array of Material Interface Object References=@n20.Materials) -> (then) {is=title K2Node_CallFunction}
   Event PreConstruct()
   > SetObject(self=Get CSVFileNameProperty() -> (CSVFileNameProperty:Single Property View Object Reference), NewObject=Self-Reference())
   > Set WidthOverride!(
    execute
    WidthOverride:float=Get!(self:Scroll Box Object Reference=Get MaterialListScrollBox()) -> (ScrollbarThickness_X:double, ScrollbarThickness_Y:double)
    self:Size Box Object Reference=Get HeadersScrollWidthCompensation() -> (HeadersScrollWidthCompensation:Size Box Object Reference)
   ) -> (then, Output_Get:float)
   > Do Once()
   > n24 = GetCurrentLevel!(execute, self:Level Editor Subsystem Object Reference=Get LevelEditorSubsystem()) -> (then, ReturnValue:Level Object Reference)
   n25 = GetDisplayName(Object=GetOuterObject!(Object:Object Reference=@n24) -> (ReturnValue:Object Reference))
   > Set Export CSV Path(Export CSV Path:String=Append(A=GetProjectSavedDirectory!() -> (ReturnValue:String), B=Append(A=@n25, B=_SortedMaterials.csv))) -> (Output_Get:String)
   n26 = Event Destruct()
  ]}
 Export CSV {nodes [
   Export CSV() {is=title K2Node_FunctionEntry}
   n1 = Get Export CSV Path() -> (Export CSV Path:String)
   n2 = Get CSV String()
   > n3 = "Execute Python Command (Advanced)"!(
    execute
    PythonCommand:String="To String (Text)"(InText=Format Text!(
      Format="INVTEXT(\"f=open(\\\"{File Name}\\\", \\\"w\\\")\\r\\nf.write(\\\"\\\"\\\"{content}\\\"\\\"\\\")\\r\\nf.close()\")"
      File Name:String=@n1
      content:String=@n2
     ) -> (Result))
    ExecutionMode:EPythonCommandExecutionMode Enum=ExecuteFile
    FileExecutionScope:EPythonFileExecutionScope Enum=Private
   ) -> (then, CommandResult:String, LogOutput:Array of Python Log Output Entry Structures, ReturnValue:bool)
   > Show Message Dialog(
    Title=MakeLiteralText(Value="INVTEXT(\"CSV Export\")")
    Message=Select(
     Option 0:"Text (by ref)"="To Text (String)"(InString=@n3)
     Option 1:"Text (by ref)"=Format Text!(Format="INVTEXT(\"Successfuly exported to CSV\\r\\n{File Name}\")", File Name=@n1) -> (Result)
     Index=@n3.ReturnValue
    ) -> (ReturnValue:"Text (by ref)")
    MessageType=Ok
    DefaultValue=No
    MessageCategory=Select(Option 0:EAppMsgCategory Enum=Error, Option 1:EAppMsgCategory Enum=Success, Index=@n3.ReturnValue) -> (ReturnValue:EAppMsgCategory Enum)
   )
  ]}
 FetchNextActorMaterials {nodes [
   FetchNextActorMaterials() {is=title K2Node_FunctionEntry}
   n1 = Get LevelActors() -> (LevelActors:Array of Actor Object References)
   > n2 = Branch(Condition="integer < integer"(A=Get CurrentActorIndex() -> (CurrentActorIndex:int=0), B=Length(TargetArray:Array of Actor Object References=@n1)))
   > Set Actor(Actor:Actor Object Reference="Get (a copy)"(Array:Array of Actor Object References=Get LevelActors(), Dimension 1=Get CurrentActorIndex() -> (CurrentActorIndex=0)) -> (Output:Actor Object Reference)) -> (Output_Get:Actor Object Reference)
   > n3 = Sequence()
   > n4 = Set BuildingActor(BuildingActor:Building Level Instance Object Reference=Cast To BuildingLevelInstance!(Object=Get Actor() -> (Actor:Actor Object Reference)) -> (AsBuilding Level Instance:Building Level Instance Object Reference, bSuccess:bool=~) {is=title K2Node_DynamicCast}) -> (Output_Get:Building Level Instance Object Reference)
   > n5 = Is Valid(InputObject=@n4)
   > n6 = GetAllSubActors!(execute, self, BuildingActor:Building Level Instance Object Reference=@n4) -> (then, ReturnValue:Array of Actor Object References)
   > n7 = For Each Loop(Array:Array of Actor Object References=@n6) -> (Array Element:Actor Object Reference)
   > n8 = Set Actor(Actor=@n7)
   n9 = For Each Loop(
    Exec=[@n8.then, @n5.Is Not Valid]
    Array:Array of Mesh Component Object References=Get Components By Class!(self:Actor Object Reference=Get Actor(), ComponentClass:Actor Component Class Reference=/Script/Engine.MeshComponent) -> (ReturnValue:Array of Mesh Component Object References)
   ) -> (Array Element:Mesh Component Object Reference)
   > n10 = For Each Loop(Array:Array of Material Interface Object References=GetMaterials!(self:Mesh Component Object Reference=@n9) -> (ReturnValue:Array of Material Interface Object References)) -> (Array Element:Material Interface Object Reference)
   > Set Material(Material:Material Interface Object Reference=@n10) -> (Output_Get:Material Interface Object Reference)
   n11 = Find!(TargetMap=Get LevelMaterialInfo(), Key=Get Material() -> (Material:Material Interface Object Reference)) -> (
    Value_Actors_6_#1:Array of Actor Object References
    Value_BuildingActors_14_#2:Array of Building Level Instance Object References
    Value_Statistics_10_#3:Material Statistics Structure
    ReturnValue
   )
   > n12 = Branch(Condition=@n11.ReturnValue)
   > n13 = Set DummyActorsArray(DummyActorsArray:Array of Actor Object References=@n11) -> (Output_Get:Array of Actor Object References)
   > Add(TargetArray:Array of Actor Object References=@n13, NewItem:"Actor Object Reference (by ref)"=Get Actor()) {is=title K2Node_CallArrayFunction}
   > Set DummyBuildingActorsArray(DummyBuildingActorsArray:Array of Building Level Instance Object References=@n11.Value_BuildingActors_14_#2) -> (Output_Get:Array of Building Level Instance Object References)
   > n14 = Is Valid(InputObject=Get BuildingActor() -> (BuildingActor:Building Level Instance Object Reference))
   > n15 = Add(
    TargetArray:Array of Building Level Instance Object References=Get DummyBuildingActorsArray() -> (DummyBuildingActorsArray:Array of Building Level Instance Object References)
    NewItem:"Building Level Instance Object Reference (by ref)"=Get BuildingActor()
   )
   Add!(
    execute=[@n15.then, @n14.Is Not Valid]
    TargetMap:Map of Material Interface Object References to Sort Tool Level Material Info Struct Structures=Get LevelMaterialInfo()
    Key:"Material Interface Object Reference (by ref)"=Get Material()
    Value:"Sort Tool Level Material Info Struct Structure (by ref)"=Make Sort Tool Level Material Info Struct!(
     Actors_6_#1:Array of Actor Object References=Get DummyActorsArray() -> (DummyActorsArray:Array of Actor Object References)
     BuildingActors_14_#2:Array of Building Level Instance Object References=Get DummyBuildingActorsArray()
     Statistics_10_#3:Material Statistics Structure=@n11.Value_Statistics_10_#3
    ) -> (SortTool_LevelMaterialInfoStruct:Sort Tool Level Material Info Struct Structure) {is=title K2Node_MakeStruct}
   ) -> (then) {is=title K2Node_CallFunction}
   n16 = GetStatistics!(execute=@n12.else, Material:Material Interface Object Reference=Get Material()) -> (then, ReturnValue:Material Statistics Structure)
   n17 = Get Actor()
   > Add(
    TargetMap=Get LevelMaterialInfo()
    Key=Get Material()
    Value=Make Sort Tool Level Material Info Struct(
     Actors_6_#1=Make Array("[0]":Actor Object Reference=@n17) -> (Array:Array of Actor Object References)
     Statistics_10_#3=@n16
    )
   )
   Increment Int(" "=@n3.then_1, Value=Get CurrentActorIndex() -> (CurrentActorIndex=0))
   n18 = Get LevelActors()
   > Set Text(
    Text=Format Text!(
     Format="INVTEXT(\"Fetching Materials in Level ({x}/{y})\")"
     x:int=Get CurrentActorIndex() -> (CurrentActorIndex=0)
     y:int=Length(TargetArray=@n18)
    ) -> (Result)
    self:Editable Text Object Reference=Get FethingMaterialsText() -> (FethingMaterialsText:Editor Utility Editable Text Object Reference)
   )
   > AsyncFetchMaterials() {is=title K2Node_CallFunction}
   ShowMaterialsList(execute=@n2.else) {is=title K2Node_CallFunction}
  ]}
 GetAllSubActors {nodes [
   GetAllSubActors!() -> (then, BuildingActor:Building Level Instance Object Reference) {is=title K2Node_FunctionEntry}
   > n1 = Sequence()
   > n2 = GetActorsInLevel!(execute, self:Building Level Instance Object Reference=Get BuildingActor()) -> (then, ReturnValue:Array of Actor Object References)
   > Append Array!(
    execute
    TargetArray:Array of Actor Object References=Get Output() -> (Output:Array of Actor Object References)
    SourceArray:Array of Actor Object References=@n2
   ) -> (then)
   > n3 = For Each Loop(Array:Array of Actor Object References=Get Output()) -> (Array Element:Actor Object Reference)
   > n4 = Cast To BuildingLevelInstance!(execute:Exec, Object=@n3) -> (then:Exec, CastFailed:Exec, AsBuilding Level Instance)
   > n5 = GetAllSubActors(BuildingActor=@n4) {is=title K2Node_CallFunction}
   > Append Array(TargetArray=Get Output(), SourceArray=@n5)
   Return Node!(execute=@n1.then_1, ReturnValue:Array of Actor Object References=Get Output())
  ]}
 GetCachedActorReferencesFileName {nodes [
   GetCachedActorReferencesFileName() {is=title K2Node_FunctionEntry}
   > n1 = GetCurrentLevel(self=Get LevelEditorSubsystem())
   n2 = GetDisplayName(Object=GetOuterObject(Object=@n1))
   > Return Node(ReturnValue:String=Append(A=GetProjectSavedDirectory(), B=Append(A=@n2, B=_CachedMaterialStatistics.cache)))
  ]}
 GetMaterialStatistic {nodes [
   GetMaterialStatistic!() -> (then, Material:Material Interface Object Reference, Statistic:SortTool_SortType Enum=NewEnumerator0) {is=title K2Node_FunctionEntry}
   > n1 = Switch on SortTool_SortType!(execute, Selection:SortTool_SortType Enum=Get Statistic() -> (Statistic:SortTool_SortType Enum=NewEnumerator0)) -> (
    NewEnumerator0:Exec
    NewEnumerator1:Exec
    NewEnumerator2:Exec
    NewEnumerator3:Exec
    NewEnumerator4:Exec
    NewEnumerator5:Exec
    NewEnumerator6:Exec
    NewEnumerator7:Exec
    NewEnumerator8:Exec
   ) {is=title K2Node_SwitchEnum}
   n2 = Find!(TargetMap=Get LevelMaterialInfo(), Key=Get Material()) -> (
    Value_Actors_6_#1
    Value_BuildingActors_14_#2
    Value_Statistics_10_#3_NumVertexShaderInstructions:int
    Value_Statistics_10_#3_NumPixelShaderInstructions:int
    Value_Statistics_10_#3_NumSamplers:int
    Value_Statistics_10_#3_NumVertexTextureSamples:int
    Value_Statistics_10_#3_NumPixelTextureSamples:int
    Value_Statistics_10_#3_NumVirtualTextureSamples:int
    Value_Statistics_10_#3_NumUVScalars:int
    Value_Statistics_10_#3_NumInterpolatorScalars:int
    ReturnValue
   )
   > Return Node(ReturnValue:int=@n2.Value_Statistics_10_#3_NumVertexShaderInstructions)
   Return Node(execute=@n1.NewEnumerator1, ReturnValue=@n2.Value_Statistics_10_#3_NumPixelShaderInstructions)
   Return Node(execute=@n1.NewEnumerator2, ReturnValue=@n2.Value_Statistics_10_#3_NumSamplers)
   Return Node(execute=@n1.NewEnumerator3, ReturnValue=@n2.Value_Statistics_10_#3_NumVertexTextureSamples)
   Return Node(execute=@n1.NewEnumerator4, ReturnValue=@n2.Value_Statistics_10_#3_NumPixelTextureSamples)
   Return Node(execute=@n1.NewEnumerator5, ReturnValue=@n2.Value_Statistics_10_#3_NumVirtualTextureSamples)
   Return Node(execute=@n1.NewEnumerator6, ReturnValue=@n2.Value_Statistics_10_#3_NumUVScalars)
   Return Node(execute=@n1.NewEnumerator7, ReturnValue=@n2.Value_Statistics_10_#3_NumInterpolatorScalars)
   Return Node(execute=@n1.NewEnumerator8, ReturnValue=Length(TargetArray=@n2))
  ]}
 GetSortedIndices {nodes [
   n1 = GetSortedIndices!() -> (then, ValuesArray:Array of Integers) {is=title K2Node_FunctionEntry}
   > n2 = For Each Loop(Array:Array of Integers=@n1) -> (Array Element:int)
   > Add(TargetArray:Array of Integers=Get Unordered() -> (Unordered:Array of Integers), NewItem:"Integer (by ref)"=@n2) {is=title K2Node_CallArrayFunction}
   Sort Integer Array!(
    execute=@n2.Completed
    TargetArray:Array of Integers=Get ValuesArray() -> (ValuesArray:Array of Integers)
    bStableSort:bool=false
    SortOrder:EArraySortOrder Enum=Descending
   ) -> (then)
   > n3 = For Each Loop(Array=Get ValuesArray())
   > Set Value(Value:int=@n3) -> (Output_Get:int=0)
   > Set Index(Index:int=0) -> (Output_Get:int=0)
   > Set Found(Found:bool=false) -> (Output_Get:bool)
   > n4 = While Loop!(execute:Exec, Condition:bool=NOT Boolean(A=Get Found() -> (Found:bool))) -> (LoopBody:Exec, Completed:Exec) {is=title K2Node_MacroInstance}
   > n5 = Branch(Condition=Contains Item(TargetArray:Array of Integers=Get Output() -> (Output:Array of Integers), ItemToFind:"Integer (by ref)"=Get Index() -> (Index:int=0)))
   > Increment Int(Value=Get Index() -> (Index=0))
   n6 = Get Index() -> (Index=0)
   n7 = Get Unordered()
   n8 = Branch(execute=@n5.else, Condition="Equal (Integer)"(A="Get (a ref)"(Array=@n7, Dimension 1=@n6), B=Get Value() -> (Value:int=0)))
   > Set Found(Found=true)
   Increment Int(" "=@n8.else, Value=Get Index() -> (Index=0))
   Add(execute=@n4.Completed, TargetArray=Get Output(), NewItem=Get Index() -> (Index=0))
   Return Node!(execute=@n3.Completed, SortedIndices:Array of Integers=Get Output())
  ]}
 ImportCache {nodes [
   ImportCache() {is=title K2Node_FunctionEntry}
   > n1 = Sequence()
   > Clear(TargetMap=Get LevelMaterialInfo())
   n2 = GetCachedActorReferencesFileName!(self) -> (ReturnValue:String) {is=title K2Node_CallFunction}
   n3 = "Execute Python Command (Advanced)"(
    execute=@n1.then_1
    PythonCommand="To String (Text)"(InText=Format Text!(Format="INVTEXT(\"f=open(\\\"{File Name}\\\", \\\"r\\\")\\r\\ncontent=f.read()\\r\\nprint(content)\\r\\nf.close()\")", File Name:String=@n2) -> (Result))
    ExecutionMode=ExecuteFile
    FileExecutionScope=Private
   )
   > n4 = Branch(Condition=@n3.ReturnValue)
   > n5 = Sequence()
   > n6 = Set Actor Cache Rows(Actor Cache Rows:Array of Strings=ParseIntoArray(
     SourceString="Get (a copy)"!(Array:Array of Python Log Output Entry Structures=@n3.LogOutput, Dimension 1) -> (Output_Type:EPythonLogOutputType Enum, Output_Output:String).Output_Output
     Delimiter="\r\n"
    )) -> (Output_Get:Array of Strings)
   > n7 = For Each Loop(Array:Array of Strings=@n6) -> (Array Element:String)
   > Clear(TargetArray:Array of Actor Object References=Get CachedActors() -> (CachedActors:Array of Actor Object References)) {is=title K2Node_CallArrayFunction}
   > Clear(TargetArray:Array of Building Level Instance Object References=Get CachedBuildingActors() -> (CachedBuildingActors:Array of Building Level Instance Object References))
   > n8 = Sequence()
   n9 = ParseIntoArray(SourceString=@n7, Delimiter=|)
   n10 = "Get (a copy)"!(Array:Array of Strings=@n9, Dimension 1) -> (Output:String)
   > n11 = LoadAsset_Blocking(Asset=To Soft Object Reference(SoftObjectPath=MakeSoftObjectPath!(PathString:String=@n10) -> (ReturnValue:Soft Object Path Structure)))
   > Set Material(Material=Cast To MaterialInterface!(Object=@n11) -> (AsMaterial Interface:Material Interface Object Reference, bSuccess:bool=~) {is=title K2Node_DynamicCast})
   n12 = For Each Loop(Exec=@n8.then_1, Array=ParseIntoArray(SourceString="Get (a copy)"(Array=@n9, Dimension 1=2), Delimiter=","))
   n13 = For Each Loop(Exec=@n12.Completed, Array=ParseIntoArray(SourceString="Get (a copy)"(Array=@n9, Dimension 1=3), Delimiter=","))
   Is Valid(exec=@n13.Completed, InputObject=Get Material())
   n14 = ParseIntoArray(SourceString="Get (a copy)"(Array=@n9, Dimension 1=1), Delimiter=",")
   n15 = String To Integer(InString="Get (a copy)"(Array=@n14, Dimension 1=5))
   n16 = String To Integer(InString="Get (a copy)"(Array=@n14, Dimension 1=6))
   n17 = String To Integer(InString="Get (a copy)"(Array=@n14))
   n18 = String To Integer(InString="Get (a copy)"(Array=@n14, Dimension 1=1))
   n19 = String To Integer(InString="Get (a copy)"(Array=@n14, Dimension 1=2))
   n20 = String To Integer(InString="Get (a copy)"(Array=@n14, Dimension 1=3))
   n21 = String To Integer(InString="Get (a copy)"(Array=@n14, Dimension 1=4))
   n22 = String To Integer(InString="Get (a copy)"(Array=@n14, Dimension 1=7))
   > Add(
    TargetMap=Get LevelMaterialInfo()
    Key=Get Material()
    Value=Make Sort Tool Level Material Info Struct(
     Actors_6_#1=Get CachedActors()
     BuildingActors_14_#2=Get CachedBuildingActors()
     Statistics_10_#3=Make Material Statistics!(
      NumVertexShaderInstructions:int=@n17
      NumPixelShaderInstructions:int=@n18
      NumSamplers:int=@n19
      NumVertexTextureSamples:int=@n20
      NumPixelTextureSamples:int=@n21
      NumVirtualTextureSamples:int=@n15
      NumUVScalars:int=@n16
      NumInterpolatorScalars:int=@n22
     ) -> (MaterialStatistics:Material Statistics Structure) {is=title K2Node_MakeStruct}
    )
   ) {is=title K2Node_CallFunction}
   n23 = LoadAsset_Blocking(execute=@n13.LoopBody, Asset=To Soft Object Reference(SoftObjectPath=MakeSoftObjectPath(PathString=@n13)))
   > n24 = Cast To BuildingLevelInstance(Object=@n23)
   > Add(
    TargetArray:Array of Building Level Instance Object References=Get CachedBuildingActors()
    NewItem:"Building Level Instance Object Reference (by ref)"=@n24
   ) {is=title K2Node_CallArrayFunction}
   n25 = LoadAsset_Blocking(execute=@n12.LoopBody, Asset=To Soft Object Reference(SoftObjectPath=MakeSoftObjectPath(PathString=@n12)))
   > n26 = Cast To Actor!(execute, Object=@n25) -> (then, CastFailed, AsActor:Actor Object Reference) {is=title K2Node_DynamicCast}
   > Add(TargetArray:Array of Actor Object References=Get CachedActors(), NewItem:"Actor Object Reference (by ref)"=@n26)
   Return Node!(execute=@n5.then_1, ReturnValue:bool=true)
   Show Message Dialog(
    execute=@n4.else
    Title=MakeLiteralText(Value="INVTEXT(\"Cache Import Error\")")
    Message=MakeLiteralText(Value="INVTEXT(\"The file does not exist!\")")
    MessageType=Ok
    DefaultValue=No
    MessageCategory=Error
   )
   > Return Node(ReturnValue=false)
  ]}
 SaveCache {nodes [
   SaveCache() {is=title K2Node_FunctionEntry}
   > n1 = Sequence()
   > Set CacheFileName(CacheFileName:String=GetCachedActorReferencesFileName()) -> (Output_Get:String)
   > n2 = For Each Loop(Array:Array of Material Interface Object References=Get MaterialsArray()) -> (Array Element:Material Interface Object Reference)
   > Set Material(Material=@n2)
   n3 = Make Soft Reference(Input=Get Material())
   n4 = "Equal (Integer)"(A=@n2.Array Index)
   > Set CachedContent(CachedContent:String=Append!(
     A=Get CachedContent() -> (CachedContent:String)
     B=Select(Option 0:String="\r\n", Option 1:String, Index=@n4) -> (ReturnValue:String)
     C:String="To String (SoftObjectReference)"(SoftObjectReference=@n3)
    ) -> (ReturnValue)) -> (Output_Get:String)
   > n5 = Sequence!(execute) -> (then_0, then_1, then_2:Exec)
   n6 = Find(TargetMap=Get LevelMaterialInfo(), Key=Get Material())
   n7 = Format Text!(
    Format="INVTEXT(\"{VS Instructions},{PS Instructions},{Samplers},{Vertex Texture Samples},{Pixel Texture Samples},{Virtual Texture Samples},{UV Scalars},{Interpolator Scalars}\")"
    VS Instructions:int=@n6.Value_Statistics_10_#3_NumVertexShaderInstructions
    PS Instructions:int=@n6.Value_Statistics_10_#3_NumPixelShaderInstructions
    Samplers:int=@n6.Value_Statistics_10_#3_NumSamplers
    Vertex Texture Samples:int=@n6.Value_Statistics_10_#3_NumVertexTextureSamples
    Pixel Texture Samples:int=@n6.Value_Statistics_10_#3_NumPixelTextureSamples
    Virtual Texture Samples:int=@n6.Value_Statistics_10_#3_NumVirtualTextureSamples
    UV Scalars:int=@n6.Value_Statistics_10_#3_NumUVScalars
    Interpolator Scalars:int=@n6.Value_Statistics_10_#3_NumInterpolatorScalars
   ) -> (Result)
   > Set CachedContent(CachedContent=Append(A=Get CachedContent(), B=|, C="To String (Text)"(InText=@n7)))
   Set CachedContent(execute=@n5.then_1, CachedContent=Append!(A=Get CachedContent(), B=|) -> (ReturnValue))
   > n8 = For Each Loop(Array:Array of Actor Object References=@n6) -> (Array Element:Actor Object Reference)
   n9 = Make Soft Reference(Input:Actor Object Reference=@n8) -> (Output:Actor Soft Object Reference)
   n10 = "Equal (Integer)"(A=@n8.Array Index)
   > Set CachedContent(CachedContent=Append!(A=Get CachedContent(), B=Select(Option 0=",", Index=@n10), C:String="To String (SoftObjectReference)"(SoftObjectReference=@n9)) -> (ReturnValue))
   Set CachedContent(execute=@n5.then_2, CachedContent=Append!(A=Get CachedContent(), B=|) -> (ReturnValue))
   > n11 = For Each Loop(Array:Array of Building Level Instance Object References=@n6.Value_BuildingActors_14_#2) -> (Array Element:Building Level Instance Object Reference)
   n12 = Make Soft Reference(Input:Building Level Instance Object Reference=@n11) -> (Output:Building Level Instance Soft Object Reference)
   n13 = "Equal (Integer)"(A=@n11.Array Index)
   > Set CachedContent(CachedContent=Append!(A=Get CachedContent(), B=Select(Option 0=",", Index=@n13), C:String="To String (SoftObjectReference)"(SoftObjectReference=@n12)) -> (ReturnValue))
   n14 = Get CachedContent()
   n15 = Get CacheFileName() -> (CacheFileName:String)
   n16 = "Execute Python Command (Advanced)"(
    execute=@n1.then_1
    PythonCommand="To String (Text)"(InText=Format Text!(
      Format="INVTEXT(\"f=open(\\\"{File Name}\\\", \\\"w\\\")\\r\\nf.write(\\\"\\\"\\\"{content}\\\"\\\"\\\")\\r\\nf.close()\")"
      File Name:String=@n15
      content:String=@n14
     ) -> (Result))
    ExecutionMode=ExecuteFile
    FileExecutionScope=Private
   )
   n17 = Get CacheFileName()
   > Show Message Dialog(
    Title=MakeLiteralText(Value="INVTEXT(\"Cache Export\")")
    Message=Select(
     Option 0:"Text (by ref)"="To Text (String)"(InString=@n16)
     Option 1:"Text (by ref)"=Format Text!(Format="INVTEXT(\"Successfuly exported to file\\r\\n{File Name}\")", File Name=@n17) -> (Result)
     Index=@n16.ReturnValue
    ) -> (ReturnValue:"Text (by ref)")
    MessageType=Ok
    DefaultValue=No
    MessageCategory=Select(Option 0:EAppMsgCategory Enum=Error, Option 1:EAppMsgCategory Enum=Success, Index=@n16.ReturnValue) -> (ReturnValue:EAppMsgCategory Enum)
   )
  ]}
 SelectActorsFromMaterialRow {nodes [
   n1 = SelectActorsFromMaterialRow!() -> (then, MaterialRow:Sort Level Materials Row Object Reference) {is=title K2Node_FunctionEntry}
   n2 = Get!(self:Sort Level Materials Row Object Reference=@n1) -> (
    Info_Actors_6_#1:Array of Actor Object References
    Info_BuildingActors_14_#2:Array of Building Level Instance Object References
    Info_Statistics_10_#3:Material Statistics Structure
   )
   > n3 = For Each Loop(Array:Array of Actor Object References=@n2) -> (Array Element:Actor Object Reference)
   n4 = For Each Loop(Exec=@n3.Completed, Array:Array of Building Level Instance Object References=@n2.Info_BuildingActors_14_#2) -> (Array Element:Building Level Instance Object Reference)
   SetSelectedLevelActors(
    execute=@n4.Completed
    self=Get EditorActorSubsystem()
    ActorsToSelect=Get ActorsToSelect() -> (ActorsToSelect:Array of Actor Object References)
   )
   Add(execute=@n4.LoopBody, TargetArray=Get ActorsToSelect(), NewItem=@n4)
   Add(execute=@n3.LoopBody, TargetArray=Get ActorsToSelect(), NewItem=@n3)
  ]}
 SortMaterials {nodes [
   SortMaterials!() -> (then, SortBy:SortTool_SortType Enum=NewEnumerator0) {is=title K2Node_FunctionEntry}
   > n1 = Sequence!(execute) -> (then_0, then_1)
   > n2 = For Each Loop(Array:Array of Material Interface Object References=Get MaterialsArray()) -> (Array Element:Material Interface Object Reference)
   > n3 = GetMaterialStatistic!(
    execute
    self
    Material:Material Interface Object Reference=@n2
    Statistic:SortTool_SortType Enum=Get SortBy() -> (SortBy:SortTool_SortType Enum=NewEnumerator0)
   ) -> (then, ReturnValue:int=0) {is=title K2Node_CallFunction}
   > Add(TargetArray:Array of Integers=Get ValuesArray(), NewItem:"Integer (by ref)"=@n3)
   n4 = GetSortedIndices!(execute=@n1.then_1, self, ValuesArray:Array of Integers=Get ValuesArray()) -> (then, SortedIndices:Array of Integers) {is=title K2Node_CallFunction}
   > n5 = For Each Loop(Array:Array of Integers=@n4) -> (Array Element:int)
   Return Node!(
    execute=@n5.Completed
    SortedMaterials:Array of Material Interface Object References=Get Output() -> (Output:Array of Material Interface Object References)
   )
   > Add(
    TargetArray:Array of Material Interface Object References=Get Output()
    NewItem:"Material Interface Object Reference (by ref)"="Get (a copy)"(Array:Array of Material Interface Object References=Get MaterialsArray(), Dimension 1=@n5) -> (Output:Material Interface Object Reference)
   )
  ]}
]
variables [
 bShouldStop {type=boolean}
 CSV String {type=string, default=""}
 CurrentActorIndex {type=integer}
 Export CSV Path {type=string, default=""}
 LevelActors {type="array<Actor>"}
 LevelMaterialInfo {default {}}
 MaterialsArray {type="array<MaterialInterface>", default []}
 ParentBuildingActors {default {}}
]
widgets {
 info {rootWidgetClass=&WidgetSwitcher, widgetCount=52}
 widgets [
  RootSwitcher {widgetClassPath=&WidgetSwitcher, bIsVariable=true, widget=RootSwitcher}
  StartingOptionsSizeBox {
   widgetClassPath=&SizeBox
   bIsVariable=true
   parent=RootSwitcher
   widget=StartingOptionsSizeBox
   properties {bOverride_WidthOverride=true, widthOverride=600}
   slot {class=WidgetSwitcherSlot, properties {horizontalAlignment=HAlign_Center, verticalAlignment=VAlign_Center}}
  }
  StartingOptionsVerticalBox {widgetClassPath=&VerticalBox, parent=StartingOptionsSizeBox, widget=StartingOptionsVerticalBox, slot {class=SizeBoxSlot}}
  FetchMaterialsButton {
   widgetClassPath=&EditorUtilityButton
   parent=StartingOptionsVerticalBox
   widget=FetchMaterialsButton
   slot {class=VerticalBoxSlot, properties {padding=<ltrb 10 10 10 10>}}
  }
  FetchMaterialsText {
   widgetClassPath=&TextBlock
   parent=FetchMaterialsButton
   widget=FetchMaterialsText
   properties {
    text=Fetch Material Statistics
    font {
     bForceMonospaced=false
     bMaterialIsStencil=false
     fontMaterial=null
     fontObject=&Roboto
     letterSpacing=0
     monospacedWidth=1
     size=15
     skewAmount=0
     typefaceFontName=Regular
     outlineSettings {
      bApplyOutlineToDropShadows=false
      bMiteredCorners=false
      bSeparateFillAlpha=false
      outlineMaterial=null
      outlineSize=0
      outlineColor=<rgba 0 0 0 1>
     }
    }
   }
   slot {class=ButtonSlot, properties {padding=<ltrb 10 10 10 10>}}
  }
  VerticalBox_2 {
   widgetClassPath=&VerticalBox
   parent=StartingOptionsVerticalBox
   widget=VerticalBox_2
   slot {class=VerticalBoxSlot, properties {padding=<ltrb 10 10 10 10>}}
  }
  LoadCachedResultsButton {widgetClassPath=&EditorUtilityButton, parent=VerticalBox_2, widget=LoadCachedResultsButton, slot {class=VerticalBoxSlot}}
  LoadCachedResultsText {
   widgetClassPath=&TextBlock
   parent=LoadCachedResultsButton
   widget=LoadCachedResultsText
   properties {
    text=Load Cached Results
    font {
     bForceMonospaced=false
     bMaterialIsStencil=false
     fontMaterial=null
     fontObject=&Roboto
     letterSpacing=0
     monospacedWidth=1
     size=15
     skewAmount=0
     typefaceFontName=Regular
     outlineSettings {
      bApplyOutlineToDropShadows=false
      bMiteredCorners=false
      bSeparateFillAlpha=false
      outlineMaterial=null
      outlineSize=0
      outlineColor=<rgba 0 0 0 1>
     }
    }
   }
   slot {class=ButtonSlot, properties {padding=<ltrb 10 10 10 10>}}
  }
  FetchingMaterialsVerticalBox {
   widgetClassPath=&VerticalBox
   bIsVariable=true
   parent=RootSwitcher
   widget=FetchingMaterialsVerticalBox
   slot {class=WidgetSwitcherSlot, properties {horizontalAlignment=HAlign_Center, verticalAlignment=VAlign_Center}}
  }
  FethingMaterialsText {
   widgetClassPath=/Script/Blutility.EditorUtilityEditableText
   bIsVariable=true
   parent=FetchingMaterialsVerticalBox
   widget=FethingMaterialsText
   properties {
    isReadOnly=true
    text=Fetching Materials in Level
    widgetStyle {
     backgroundImageComposing {
      bIsDynamicallyLoaded=false
      drawAs=Border
      imageType=FullColor
      mirroring=NoMirror
      resourceName=../../../Engine/Content/Slate/Old/HyperlinkDotted.png
      resourceObject=null
      tiling=Both
      imageSize=<0 0>
      margin=<ltrb 0 0 0 0.1875>
      outlineSettings {
       bUseBrushTransparency=false
       roundingType=HalfHeightRadius
       width=0
       color {colorUseRule=UseColor_Specified, specifiedColor=<rgba 0 0 0 0>}
       cornerRadii=<0 0 0 0>
      }
      tintColor {colorUseRule=UseColor_Specified, specifiedColor=<rgba 1 1 1 1>}
      uVRegion {bIsValid=false, max=<0 0>, min=<0 0>}
     }
     backgroundImageSelected {
      bIsDynamicallyLoaded=false
      drawAs=Image
      imageType=NoImage
      mirroring=NoMirror
      resourceName=None
      resourceObject=null
      tiling=NoTile
      imageSize=<0 0>
      margin=<ltrb 0 0 0 0>
      outlineSettings {
       bUseBrushTransparency=false
       roundingType=HalfHeightRadius
       width=0
       color {colorUseRule=UseColor_Specified, specifiedColor=<rgba 0 0 0 0>}
       cornerRadii=<0 0 0 0>
      }
      tintColor {colorUseRule=UseColor_Specified, specifiedColor=<rgba 0 0.162 0.7454 1>}
      uVRegion {bIsValid=false, max=<0 0>, min=<0 0>}
     }
     caretImage {
      bIsDynamicallyLoaded=false
      drawAs=Image
      imageType=NoImage
      mirroring=NoMirror
      resourceName=None
      resourceObject=null
      tiling=NoTile
      imageSize=<0 0>
      margin=<ltrb 0 0 0 0>
      outlineSettings {
       bUseBrushTransparency=false
       roundingType=HalfHeightRadius
       width=0
       color {colorUseRule=UseColor_Specified, specifiedColor=<rgba 0 0 0 0>}
       cornerRadii=<0 0 0 0>
      }
      tintColor {colorUseRule=UseColor_Specified, specifiedColor=<rgba 1 1 1 1>}
      uVRegion {bIsValid=false, max=<0 0>, min=<0 0>}
     }
     colorAndOpacity {colorUseRule=UseColor_Foreground, specifiedColor=<rgba 1 0 1 1>}
     font {
      bForceMonospaced=false
      bMaterialIsStencil=false
      fontMaterial=null
      fontObject=null
      letterSpacing=0
      monospacedWidth=1
      size=15
      skewAmount=0
      typefaceFontName=Regular
      outlineSettings {
       bApplyOutlineToDropShadows=false
       bMiteredCorners=false
       bSeparateFillAlpha=false
       outlineMaterial=null
       outlineSize=0
       outlineColor=<rgba 0 0 0 1>
      }
     }
     hintColorAndOpacity {colorUseRule=UseColor_Foreground, specifiedColor=<rgba 1 0 1 1>}
     selectedBackgroundColor {colorUseRule=UseColor_Foreground, specifiedColor=<rgba 1 0 1 1>}
    }
   }
   slot {class=VerticalBoxSlot}
  }
  FetchingMaterialsThrobber {
   widgetClassPath=/Script/Blutility.EditorUtilityThrobber
   parent=FetchingMaterialsVerticalBox
   widget=FetchingMaterialsThrobber
   slot {class=VerticalBoxSlot, properties {horizontalAlignment=HAlign_Center}}
  }
  CancelButton {
   widgetClassPath=&EditorUtilityButton
   parent=FetchingMaterialsVerticalBox
   widget=CancelButton
   slot {class=VerticalBoxSlot, properties {horizontalAlignment=HAlign_Center}}
  }
  Image_152 {
   widgetClassPath=&Image
   parent=CancelButton
   widget=Image_152
   properties {brush {
     bIsDynamicallyLoaded=false
     drawAs=Image
     imageType=FullColor
     mirroring=NoMirror
     resourceName=None
     resourceObject=null
     tiling=NoTile
     imageSize=<16 16>
     margin=<ltrb 0 0 0 0>
     outlineSettings {
      bUseBrushTransparency=false
      roundingType=HalfHeightRadius
      width=0
      color {colorUseRule=UseColor_Specified, specifiedColor=<rgba 0 0 0 0>}
      cornerRadii=<0 0 0 0>
     }
     tintColor {colorUseRule=UseColor_Specified, specifiedColor=<rgba 1 1 1 1>}
     uVRegion {bIsValid=false, max=<0 0>, min=<0 0>}
    }}
   slot {class=ButtonSlot}
  }
  MaterialListVerticalBox {
   widgetClassPath=&VerticalBox
   bIsVariable=true
   parent=RootSwitcher
   widget=MaterialListVerticalBox
   slot {class=WidgetSwitcherSlot}
  }
  RestartButton {
   widgetClassPath=&EditorUtilityButton
   parent=MaterialListVerticalBox
   widget=RestartButton
   slot {class=VerticalBoxSlot, properties {horizontalAlignment=HAlign_Left}}
  }
  Image_1 {
   widgetClassPath=&Image
   parent=RestartButton
   widget=Image_1
   properties {brush {
     bIsDynamicallyLoaded=false
     drawAs=Image
     imageType=FullColor
     mirroring=NoMirror
     resourceName=None
     resourceObject=/EditorDebugMenu/Textures/T_Refresh
     tiling=NoTile
     imageSize=<20 20>
     margin=<ltrb 0 0 0 0>
     outlineSettings {
      bUseBrushTransparency=false
      roundingType=HalfHeightRadius
      width=0
      color {colorUseRule=UseColor_Specified, specifiedColor=<rgba 0 0 0 0>}
      cornerRadii=<0 0 0 0>
     }
     tintColor {colorUseRule=UseColor_Specified, specifiedColor=<rgba 1 1 1 1>}
     uVRegion {bIsValid=false, max=<0 0>, min=<0 0>}
    }}
   slot {class=ButtonSlot}
  }
  HorizontalBox_81 {widgetClassPath=&HorizontalBox, parent=MaterialListVerticalBox, widget=HorizontalBox_81, slot {class=VerticalBoxSlot}}
  VerticalDivider_8 {
   widgetClassPath=&Image
   parent=HorizontalBox_81
   widget=VerticalDivider_8
   properties {brush {
     bIsDynamicallyLoaded=false
     drawAs=Image
     imageType=NoImage
     mirroring=NoMirror
     resourceName=None
     resourceObject=null
     tiling=NoTile
     imageSize=<1 32>
     margin=<ltrb 0 0 0 0>
     outlineSettings {
      bUseBrushTransparency=false
      roundingType=HalfHeightRadius
      width=0
      color {colorUseRule=UseColor_Specified, specifiedColor=<rgba 0 0 0 0>}
      cornerRadii=<0 0 0 0>
     }
     tintColor {colorUseRule=UseColor_Specified, specifiedColor=<rgba 0.0284 0.0284 0.0284 1>}
     uVRegion {bIsValid=false, max=<0 0>, min=<0 0>}
    }}
   slot {class=HorizontalBoxSlot, properties {horizontalAlignment=HAlign_Center}}
  }
  AffectedObjectsLabelOverlay {
   widgetClassPath=&Overlay
   parent=HorizontalBox_81
   widget=AffectedObjectsLabelOverlay
   slot {class=HorizontalBoxSlot, properties {padding=<ltrb 1 1 1 1>, size {sizeRule=Fill, value=1}}}
  }
  Image_109 {
   widgetClassPath=&Image
   parent=AffectedObjectsLabelOverlay
   widget=Image_109
   properties {colorAndOpacity=<rgba 0.0284 0.0284 0.0284 1>}
   slot {class=OverlaySlot, properties {horizontalAlignment=HAlign_Fill, verticalAlignment=VAlign_Fill}}
  }
  TextBlock_1 {
   widgetClassPath=&TextBlock
   parent=AffectedObjectsLabelOverlay
   widget=TextBlock_1
   properties {
    text=Material
    colorAndOpacity {colorUseRule=UseColor_Specified, specifiedColor=<rgba 0.5271 0.5271 0.5271 1>}
    font {
     bForceMonospaced=false
     bMaterialIsStencil=false
     fontMaterial=null
     fontObject=&Roboto
     letterSpacing=0
     monospacedWidth=1
     size=9
     skewAmount=0
     typefaceFontName=Bold
     outlineSettings {
      bApplyOutlineToDropShadows=false
      bMiteredCorners=false
      bSeparateFillAlpha=false
      outlineMaterial=null
      outlineSize=0
      outlineColor=<rgba 0 0 0 1>
     }
    }
   }
   slot {class=OverlaySlot, properties {horizontalAlignment=HAlign_Center, verticalAlignment=VAlign_Center, padding=<ltrb 4 2 4 2>}}
  }
  VerticalDivider_7 {
   widgetClassPath=&Image
   parent=HorizontalBox_81
   widget=VerticalDivider_7
   properties {brush {
     bIsDynamicallyLoaded=false
     drawAs=Image
     imageType=NoImage
     mirroring=NoMirror
     resourceName=None
     resourceObject=null
     tiling=NoTile
     imageSize=<1 32>
     margin=<ltrb 0 0 0 0>
     outlineSettings {
      bUseBrushTransparency=false
      roundingType=HalfHeightRadius
      width=0
      color {colorUseRule=UseColor_Specified, specifiedColor=<rgba 0 0 0 0>}
      cornerRadii=<0 0 0 0>
     }
     tintColor {colorUseRule=UseColor_Specified, specifiedColor=<rgba 0.0284 0.0284 0.0284 1>}
     uVRegion {bIsValid=false, max=<0 0>, min=<0 0>}
    }}
   slot {class=HorizontalBoxSlot, properties {horizontalAlignment=HAlign_Center}}
  }
  VSInstructionsHeader {
   widgetClassPath=&/SortLevelMaterialToolColumnHeader
   bIsVariable=true
   parent=HorizontalBox_81
   widget=VSInstructionsHeader
   properties {headerName=VS Instructions}
   slot {class=HorizontalBoxSlot, properties {size {sizeRule=Fill, value=1}}}
  }
  VerticalDivider {
   widgetClassPath=&Image
   parent=HorizontalBox_81
   widget=VerticalDivider
   properties {brush {
     bIsDynamicallyLoaded=false
     drawAs=Image
     imageType=NoImage
     mirroring=NoMirror
     resourceName=None
     resourceObject=null
     tiling=NoTile
     imageSize=<1 32>
     margin=<ltrb 0 0 0 0>
     outlineSettings {
      bUseBrushTransparency=false
      roundingType=HalfHeightRadius
      width=0
      color {colorUseRule=UseColor_Specified, specifiedColor=<rgba 0 0 0 0>}
      cornerRadii=<0 0 0 0>
     }
     tintColor {colorUseRule=UseColor_Specified, specifiedColor=<rgba 0.0284 0.0284 0.0284 1>}
     uVRegion {bIsValid=false, max=<0 0>, min=<0 0>}
    }}
   slot {class=HorizontalBoxSlot, properties {horizontalAlignment=HAlign_Center}}
  }
  PSInstructionsHeader {
   widgetClassPath=&SortLevelMaterialToolColumnHeader
   bIsVariable=true
   parent=HorizontalBox_81
   widget=PSInstructionsHeader
   properties {headerName=PS Instructions}
   slot {class=HorizontalBoxSlot, properties {size {sizeRule=Fill, value=1}}}
  }
  VerticalDivider_1 {
   widgetClassPath=&Image
   parent=HorizontalBox_81
   widget=VerticalDivider_1
   properties {brush {
     bIsDynamicallyLoaded=false
     drawAs=Image
     imageType=NoImage
     mirroring=NoMirror
     resourceName=None
     resourceObject=null
     tiling=NoTile
     imageSize=<1 32>
     margin=<ltrb 0 0 0 0>
     outlineSettings {
      bUseBrushTransparency=false
      roundingType=HalfHeightRadius
      width=0
      color {colorUseRule=UseColor_Specified, specifiedColor=<rgba 0 0 0 0>}
      cornerRadii=<0 0 0 0>
     }
     tintColor {colorUseRule=UseColor_Specified, specifiedColor=<rgba 0.0284 0.0284 0.0284 1>}
     uVRegion {bIsValid=false, max=<0 0>, min=<0 0>}
    }}
   slot {class=HorizontalBoxSlot, properties {horizontalAlignment=HAlign_Center}}
  }
  SamplersHeader {
   widgetClassPath=&SortLevelMaterialToolColumnHeader
   bIsVariable=true
   parent=HorizontalBox_81
   widget=SamplersHeader
   properties {headerName=Samplers}
   slot {class=HorizontalBoxSlot, properties {size {sizeRule=Fill, value=1}}}
  }
  VerticalDivider_2 {
   widgetClassPath=&Image
   parent=HorizontalBox_81
   widget=VerticalDivider_2
   properties {brush {
     bIsDynamicallyLoaded=false
     drawAs=Image
     imageType=NoImage
     mirroring=NoMirror
     resourceName=None
     resourceObject=null
     tiling=NoTile
     imageSize=<1 32>
     margin=<ltrb 0 0 0 0>
     outlineSettings {
      bUseBrushTransparency=false
      roundingType=HalfHeightRadius
      width=0
      color {colorUseRule=UseColor_Specified, specifiedColor=<rgba 0 0 0 0>}
      cornerRadii=<0 0 0 0>
     }
     tintColor {colorUseRule=UseColor_Specified, specifiedColor=<rgba 0.0284 0.0284 0.0284 1>}
     uVRegion {bIsValid=false, max=<0 0>, min=<0 0>}
    }}
   slot {class=HorizontalBoxSlot, properties {horizontalAlignment=HAlign_Center}}
  }
  VertexTextureSamplesHeader {
   widgetClassPath=&SortLevelMaterialToolColumnHeader
   bIsVariable=true
   parent=HorizontalBox_81
   widget=VertexTextureSamplesHeader
   properties {headerName=Vertex Texture Samples}
   slot {class=HorizontalBoxSlot, properties {size {sizeRule=Fill, value=1}}}
  }
  VerticalDivider_3 {
   widgetClassPath=&Image
   parent=HorizontalBox_81
   widget=VerticalDivider_3
   properties {brush {
     bIsDynamicallyLoaded=false
     drawAs=Image
     imageType=NoImage
     mirroring=NoMirror
     resourceName=None
     resourceObject=null
     tiling=NoTile
     imageSize=<1 32>
     margin=<ltrb 0 0 0 0>
     outlineSettings {
      bUseBrushTransparency=false
      roundingType=HalfHeightRadius
      width=0
      color {colorUseRule=UseColor_Specified, specifiedColor=<rgba 0 0 0 0>}
      cornerRadii=<0 0 0 0>
     }
     tintColor {colorUseRule=UseColor_Specified, specifiedColor=<rgba 0.0284 0.0284 0.0284 1>}
     uVRegion {bIsValid=false, max=<0 0>, min=<0 0>}
    }}
   slot {class=HorizontalBoxSlot, properties {horizontalAlignment=HAlign_Center}}
  }
  PixelTextureSamplesHeader {
   widgetClassPath=&SortLevelMaterialToolColumnHeader
   bIsVariable=true
   parent=HorizontalBox_81
   widget=PixelTextureSamplesHeader
   properties {headerName=Pixel Texture Samples}
   slot {class=HorizontalBoxSlot, properties {size {sizeRule=Fill, value=1}}}
  }
  VerticalDivider_4 {
   widgetClassPath=&Image
   parent=HorizontalBox_81
   widget=VerticalDivider_4
   properties {brush {
     bIsDynamicallyLoaded=false
     drawAs=Image
     imageType=NoImage
     mirroring=NoMirror
     resourceName=None
     resourceObject=null
     tiling=NoTile
     imageSize=<1 32>
     margin=<ltrb 0 0 0 0>
     outlineSettings {
      bUseBrushTransparency=false
      roundingType=HalfHeightRadius
      width=0
      color {colorUseRule=UseColor_Specified, specifiedColor=<rgba 0 0 0 0>}
      cornerRadii=<0 0 0 0>
     }
     tintColor {colorUseRule=UseColor_Specified, specifiedColor=<rgba 0.0284 0.0284 0.0284 1>}
     uVRegion {bIsValid=false, max=<0 0>, min=<0 0>}
    }}
   slot {class=HorizontalBoxSlot, properties {horizontalAlignment=HAlign_Center}}
  }
  VirtualTextureSamplesHeader {
   widgetClassPath=&SortLevelMaterialToolColumnHeader
   bIsVariable=true
   parent=HorizontalBox_81
   widget=VirtualTextureSamplesHeader
   properties {headerName=Virtual Texture Samples}
   slot {class=HorizontalBoxSlot, properties {size {sizeRule=Fill, value=1}}}
  }
  VerticalDivider_5 {
   widgetClassPath=&Image
   parent=HorizontalBox_81
   widget=VerticalDivider_5
   properties {brush {
     bIsDynamicallyLoaded=false
     drawAs=Image
     imageType=NoImage
     mirroring=NoMirror
     resourceName=None
     resourceObject=null
     tiling=NoTile
     imageSize=<1 32>
     margin=<ltrb 0 0 0 0>
     outlineSettings {
      bUseBrushTransparency=false
      roundingType=HalfHeightRadius
      width=0
      color {colorUseRule=UseColor_Specified, specifiedColor=<rgba 0 0 0 0>}
      cornerRadii=<0 0 0 0>
     }
     tintColor {colorUseRule=UseColor_Specified, specifiedColor=<rgba 0.0284 0.0284 0.0284 1>}
     uVRegion {bIsValid=false, max=<0 0>, min=<0 0>}
    }}
   slot {class=HorizontalBoxSlot, properties {horizontalAlignment=HAlign_Center}}
  }
  UVScalarsHeader {
   widgetClassPath=&SortLevelMaterialToolColumnHeader
   bIsVariable=true
   parent=HorizontalBox_81
   widget=UVScalarsHeader
   properties {headerName=UV Scalars}
   slot {class=HorizontalBoxSlot, properties {size {sizeRule=Fill, value=1}}}
  }
  VerticalDivider_6 {
   widgetClassPath=&Image
   parent=HorizontalBox_81
   widget=VerticalDivider_6
   properties {brush {
     bIsDynamicallyLoaded=false
     drawAs=Image
     imageType=NoImage
     mirroring=NoMirror
     resourceName=None
     resourceObject=null
     tiling=NoTile
     imageSize=<1 32>
     margin=<ltrb 0 0 0 0>
     outlineSettings {
      bUseBrushTransparency=false
      roundingType=HalfHeightRadius
      width=0
      color {colorUseRule=UseColor_Specified, specifiedColor=<rgba 0 0 0 0>}
      cornerRadii=<0 0 0 0>
     }
     tintColor {colorUseRule=UseColor_Specified, specifiedColor=<rgba 0.0284 0.0284 0.0284 1>}
     uVRegion {bIsValid=false, max=<0 0>, min=<0 0>}
    }}
   slot {class=HorizontalBoxSlot, properties {horizontalAlignment=HAlign_Center}}
  }
  InterpolatorScalarsHeader {
   widgetClassPath=&SortLevelMaterialToolColumnHeader
   bIsVariable=true
   parent=HorizontalBox_81
   widget=InterpolatorScalarsHeader
   properties {headerName=Interpolator Scalars}
   slot {class=HorizontalBoxSlot, properties {size {sizeRule=Fill, value=1}}}
  }
  VerticalDivider_9 {
   widgetClassPath=&Image
   parent=HorizontalBox_81
   widget=VerticalDivider_9
   properties {brush {
     bIsDynamicallyLoaded=false
     drawAs=Image
     imageType=NoImage
     mirroring=NoMirror
     resourceName=None
     resourceObject=null
     tiling=NoTile
     imageSize=<1 32>
     margin=<ltrb 0 0 0 0>
     outlineSettings {
      bUseBrushTransparency=false
      roundingType=HalfHeightRadius
      width=0
      color {colorUseRule=UseColor_Specified, specifiedColor=<rgba 0 0 0 0>}
      cornerRadii=<0 0 0 0>
     }
     tintColor {colorUseRule=UseColor_Specified, specifiedColor=<rgba 0.0284 0.0284 0.0284 1>}
     uVRegion {bIsValid=false, max=<0 0>, min=<0 0>}
    }}
   slot {class=HorizontalBoxSlot, properties {horizontalAlignment=HAlign_Center}}
  }
  ActorsinLevelHeader {
   widgetClassPath=&SortLevelMaterialToolColumnHeader
   bIsVariable=true
   parent=HorizontalBox_81
   widget=ActorsinLevelHeader
   properties {headerName=Actors in Level}
   slot {class=HorizontalBoxSlot, properties {size {sizeRule=Fill, value=1}}}
  }
  VerticalDivider_10 {
   widgetClassPath=&Image
   parent=HorizontalBox_81
   widget=VerticalDivider_10
   properties {brush {
     bIsDynamicallyLoaded=false
     drawAs=Image
     imageType=NoImage
     mirroring=NoMirror
     resourceName=None
     resourceObject=null
     tiling=NoTile
     imageSize=<1 32>
     margin=<ltrb 0 0 0 0>
     outlineSettings {
      bUseBrushTransparency=false
      roundingType=HalfHeightRadius
      width=0
      color {colorUseRule=UseColor_Specified, specifiedColor=<rgba 0 0 0 0>}
      cornerRadii=<0 0 0 0>
     }
     tintColor {colorUseRule=UseColor_Specified, specifiedColor=<rgba 0.0284 0.0284 0.0284 1>}
     uVRegion {bIsValid=false, max=<0 0>, min=<0 0>}
    }}
   slot {class=HorizontalBoxSlot, properties {horizontalAlignment=HAlign_Center}}
  }
  HeadersScrollWidthCompensation {
   widgetClassPath=&SizeBox
   bIsVariable=true
   parent=HorizontalBox_81
   widget=HeadersScrollWidthCompensation
   properties {bOverride_WidthOverride=true}
   slot {class=HorizontalBoxSlot}
  }
  MaterialListScrollBox {
   widgetClassPath=&EditorUtilityScrollBox
   bIsVariable=true
   parent=MaterialListVerticalBox
   widget=MaterialListScrollBox
   slot {class=VerticalBoxSlot, properties {size {sizeRule=Fill, value=1}}}
  }
  SizeBox_0 {
   widgetClassPath=&SizeBox
   parent=MaterialListVerticalBox
   widget=SizeBox_0
   properties {bOverride_WidthOverride=true, widthOverride=200}
   slot {class=VerticalBoxSlot, properties {horizontalAlignment=HAlign_Left}}
  }
  SaveCacheButton {widgetClassPath=&EditorUtilityButton, parent=SizeBox_0, widget=SaveCacheButton, slot {class=SizeBoxSlot}}
  SaveCacheText {
   widgetClassPath=&TextBlock
   parent=SaveCacheButton
   widget=SaveCacheText
   properties {
    text=Save Results in Project Cache
    colorAndOpacity {colorUseRule=UseColor_Specified, specifiedColor=<rgba 0.5271 0.5271 0.5271 1>}
    font {
     bForceMonospaced=false
     bMaterialIsStencil=false
     fontMaterial=null
     fontObject=&Roboto
     letterSpacing=0
     monospacedWidth=1
     size=9
     skewAmount=0
     typefaceFontName=Bold
     outlineSettings {
      bApplyOutlineToDropShadows=false
      bMiteredCorners=false
      bSeparateFillAlpha=false
      outlineMaterial=null
      outlineSize=0
      outlineColor=<rgba 0 0 0 1>
     }
    }
   }
   slot {class=ButtonSlot}
  }
  HorizontalBox_0 {widgetClassPath=&HorizontalBox, parent=MaterialListVerticalBox, widget=HorizontalBox_0, slot {class=VerticalBoxSlot}}
  SizeBox_1 {
   widgetClassPath=&SizeBox
   parent=HorizontalBox_0
   widget=SizeBox_1
   properties {bOverride_WidthOverride=true, widthOverride=200}
   slot {class=HorizontalBoxSlot}
  }
  ExportCSVButton {widgetClassPath=&EditorUtilityButton, parent=SizeBox_1, widget=ExportCSVButton, slot {class=SizeBoxSlot}}
  ExportCSVText {
   widgetClassPath=&TextBlock
   parent=ExportCSVButton
   widget=ExportCSVText
   properties {
    text=Export to CSV
    colorAndOpacity {colorUseRule=UseColor_Specified, specifiedColor=<rgba 0.5271 0.5271 0.5271 1>}
    font {
     bForceMonospaced=false
     bMaterialIsStencil=false
     fontMaterial=null
     fontObject=&Roboto
     letterSpacing=0
     monospacedWidth=1
     size=9
     skewAmount=0
     typefaceFontName=Bold
     outlineSettings {
      bApplyOutlineToDropShadows=false
      bMiteredCorners=false
      bSeparateFillAlpha=false
      outlineMaterial=null
      outlineSize=0
      outlineColor=<rgba 0 0 0 1>
     }
    }
   }
   slot {class=ButtonSlot}
  }
  CSVFileNameProperty {
   widgetClassPath=&SinglePropertyView
   parent=HorizontalBox_0
   widget=CSVFileNameProperty
   properties {propertyName=Export CSV Path}
   slot {class=HorizontalBoxSlot, properties {size {sizeRule=Fill, value=1}}}
  }
  NoMaterialsVerticalBox {
   widgetClassPath=&VerticalBox
   bIsVariable=true
   parent=RootSwitcher
   widget=NoMaterialsVerticalBox
   slot {class=WidgetSwitcherSlot, properties {horizontalAlignment=HAlign_Center, verticalAlignment=VAlign_Center}}
  }
  NoMaterials {
   widgetClassPath=&EditorUtilityEditableText
   bIsVariable=true
   parent=NoMaterialsVerticalBox
   widget=NoMaterials
   properties {
    isReadOnly=true
    text=No materials were found in the level
    widgetStyle {
     backgroundImageComposing {
      bIsDynamicallyLoaded=false
      drawAs=Border
      imageType=FullColor
      mirroring=NoMirror
      resourceName=../../../Engine/Content/Slate/Old/HyperlinkDotted.png
      resourceObject=null
      tiling=Both
      imageSize=<0 0>
      margin=<ltrb 0 0 0 0.1875>
      outlineSettings {
       bUseBrushTransparency=false
       roundingType=HalfHeightRadius
       width=0
       color {colorUseRule=UseColor_Specified, specifiedColor=<rgba 0 0 0 0>}
       cornerRadii=<0 0 0 0>
      }
      tintColor {colorUseRule=UseColor_Specified, specifiedColor=<rgba 1 1 1 1>}
      uVRegion {bIsValid=false, max=<0 0>, min=<0 0>}
     }
     backgroundImageSelected {
      bIsDynamicallyLoaded=false
      drawAs=Image
      imageType=NoImage
      mirroring=NoMirror
      resourceName=None
      resourceObject=null
      tiling=NoTile
      imageSize=<0 0>
      margin=<ltrb 0 0 0 0>
      outlineSettings {
       bUseBrushTransparency=false
       roundingType=HalfHeightRadius
       width=0
       color {colorUseRule=UseColor_Specified, specifiedColor=<rgba 0 0 0 0>}
       cornerRadii=<0 0 0 0>
      }
      tintColor {colorUseRule=UseColor_Specified, specifiedColor=<rgba 0 0.162 0.7454 1>}
      uVRegion {bIsValid=false, max=<0 0>, min=<0 0>}
     }
     caretImage {
      bIsDynamicallyLoaded=false
      drawAs=Image
      imageType=NoImage
      mirroring=NoMirror
      resourceName=None
      resourceObject=null
      tiling=NoTile
      imageSize=<0 0>
      margin=<ltrb 0 0 0 0>
      outlineSettings {
       bUseBrushTransparency=false
       roundingType=HalfHeightRadius
       width=0
       color {colorUseRule=UseColor_Specified, specifiedColor=<rgba 0 0 0 0>}
       cornerRadii=<0 0 0 0>
      }
      tintColor {colorUseRule=UseColor_Specified, specifiedColor=<rgba 1 1 1 1>}
      uVRegion {bIsValid=false, max=<0 0>, min=<0 0>}
     }
     colorAndOpacity {colorUseRule=UseColor_Foreground, specifiedColor=<rgba 1 0 1 1>}
     font {
      bForceMonospaced=false
      bMaterialIsStencil=false
      fontMaterial=null
      fontObject=null
      letterSpacing=0
      monospacedWidth=1
      size=15
      skewAmount=0
      typefaceFontName=Regular
      outlineSettings {
       bApplyOutlineToDropShadows=false
       bMiteredCorners=false
       bSeparateFillAlpha=false
       outlineMaterial=null
       outlineSize=0
       outlineColor=<rgba 0 0 0 1>
      }
     }
     hintColorAndOpacity {colorUseRule=UseColor_Foreground, specifiedColor=<rgba 1 0 1 1>}
     selectedBackgroundColor {colorUseRule=UseColor_Foreground, specifiedColor=<rgba 1 0 1 1>}
    }
   }
   slot {class=VerticalBoxSlot}
  }
 ]
}
